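# Physics asset for the crouch variant of the SuperNovaTaro long-hair FaceAcc part, covering back-hair dynamics on a medium female character.
# Capsules on spine, clavicles, neck and upper arms form the body volume around the hair.
# Center, left and right hair chains hang from neck_02, with side chains branching after their 03 bones.
# Hair links mostly use narrow twist and asymmetric swing limits, while body joints and center/left roots rotate freely.
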PhysicsAsset /BRCosmetics/Characters/Player/Female/Medium/Bodies/F_MED_SuperNovaTaro_Long/Meshes/Parts/F_MED_SuperNovaTaro_Long_FaceAcc_Crouch
bodies {
 clavicle_l [{shapeType=Capsule, length=5.3544, radius=5.3769, center=<6.4482 1.7189 -2.4355>, rotation=<rot 85.606 0 0>}]
 clavicle_r [{shapeType=Capsule, length=4.8637, radius=5.8676, center=<-6.4482 -0.4982 2.4355>, rotation=<rot 85.606 0 0>}]
 dyn_hair_long_bk_01 [{shapeType=Capsule, length=2, radius=2, shapeName=dyn_hair_long_bk_01_capsule, rotation=<rot 90 0 0>}]
 dyn_hair_long_bk_02 [{shapeType=Capsule, length=4.539, radius=3.9433, shapeName=dyn_hair_long_bk_02_capsule, center=<3 0 1>, rotation=<rot 90 0 0>}]
 dyn_hair_long_bk_03 [{shapeType=Capsule, length=4.4405, radius=3.7157, shapeName=dyn_hair_long_bk_03_capsule, center=<3 0 1>, rotation=<rot 90 0 0>}]
 dyn_hair_long_bk_04 [{shapeType=Capsule, length=4.3626, radius=4.783, shapeName=dyn_hair_long_bk_04_capsule, center=<3 0 0.5>, rotation=<rot 90 0 0>}]
 dyn_hair_long_bk_05 [{shapeType=Capsule, length=5.4905, radius=4.2588, shapeName=dyn_hair_long_bk_05_capsule, center=<2 0 0.5>, rotation=<rot 90 0 0>}]
 dyn_hair_long_bk_l_01 [{
   shapeType=Capsule
   length=4.0202
   radius=2.3587
   shapeName=dyn_hair_long_bk_l_01_capsule
   center=<2.7041 0.4985 0.0635>
   rotation=<rot 0 81.6548 -52.333>
  }]
 dyn_hair_long_bk_l_02 [{shapeType=Capsule, length=3.9899, radius=2.5865, shapeName=dyn_hair_long_bk_l_02_capsule, center=<3 0 2.7132>, rotation=<rot 90 0 0>}]
 dyn_hair_long_bk_l_03 [{shapeType=Capsule, length=4.9597, radius=3.6257, shapeName=dyn_hair_long_bk_l_03_capsule, center=<3 0 2.7131>, rotation=<rot 90 0 0>}]
 dyn_hair_long_bk_l_04 [{shapeType=Capsule, length=3.239, radius=3.621, shapeName=dyn_hair_long_bk_l_04_capsule, center=<3 0 2.7131>, rotation=<rot 90 0 0>}]
 dyn_hair_long_bk_l_05 [{shapeType=Capsule, length=4.971, radius=4.0825, shapeName=dyn_hair_long_bk_l_05_capsule, center=<3 0 1.7131>, rotation=<rot 90 0 0>}]
 dyn_hair_long_bk_lb_04 [{shapeType=Capsule, length=2.7842, radius=2.8474, shapeName=dyn_hair_long_bk_lb_04_capsule, center=<3 0 2.7132>, rotation=<rot 90 0 0>}]
 dyn_hair_long_bk_lb_05 [{shapeType=Capsule, length=1.8626, radius=2.5197, shapeName=dyn_hair_long_bk_lb_05_capsule, center=<3 0 1.9008>, rotation=<rot 90 0 0>}]
 dyn_hair_long_bk_r_02 [{shapeType=Capsule, length=4.101, radius=2.9777, shapeName=dyn_hair_long_bk_r_02_capsule, center=<3 0 2.5>, rotation=<rot 90 0 0>}]
 dyn_hair_long_bk_r_03 [{shapeType=Capsule, length=5.2685, radius=3.7321, shapeName=dyn_hair_long_bk_r_03_capsule, center=<3 0 2.7131>, rotation=<rot 90 0 0>}]
 dyn_hair_long_bk_r_04 [{shapeType=Capsule, length=2.3384, radius=3.5703, shapeName=dyn_hair_long_bk_r_04_capsule, center=<3 0 2.7131>, rotation=<rot 90 0 0>}]
 dyn_hair_long_bk_r_05 [{shapeType=Capsule, length=6.0655, radius=4.3041, shapeName=dyn_hair_long_bk_r_05_capsule, center=<3 0 1.7131>, rotation=<rot 90 0 0>}]
 dyn_hair_long_bk_rb_04 [{shapeType=Capsule, length=3.5602, radius=3.055, shapeName=dyn_hair_long_bk_rb_04_capsule, center=<3 0 2.7132>, rotation=<rot 90 0 0>}]
 dyn_hair_long_bk_rb_05 [{shapeType=Capsule, length=4.4851, radius=1.9305, shapeName=dyn_hair_long_bk_rb_05_capsule, center=<3 0 1.9008>, rotation=<rot 90 0 0>}]
 neck_02 [{shapeType=Capsule, length=2.5, radius=4.5, shapeName=neck_02_capsule, center=<-4.0488 0 0>, rotation=<rot 90 0 0>}]
 spine_01 [{shapeType=Capsule, length=7.0613, radius=9.6791, center=<0 1.0381 -0.0001>}]
 spine_02 [{shapeType=Capsule, length=5.9705, radius=8.255, center=<0 1.813 -0.0001>}]
 spine_03 [{shapeType=Capsule, length=7.9414, radius=7.7882, center=<0 2.0457 -0.0001>}]
 spine_04 [{shapeType=Capsule, length=9.7013, radius=7.968, center=<0.6532 1.8714 0>}]
 spine_05 [{shapeType=Capsule, length=15.0318, radius=7.5064, center=<3.2894 -0.6494 -0.0001>}]
 upperarm_l [{shapeType=Capsule, length=13.5, radius=4.5, shapeName=upperarm_l_capsule, center=<6 0 0>, rotation=<rot 90 0 0>}]
 upperarm_r [{shapeType=Capsule, length=13, radius=5, shapeName=upperarm_r_capsule, center=<-6 0 0>, rotation=<rot 90 0 0>}]
}
constraints [
 {bone1Name=dyn_hair_long_bk_l_02, bone2Name=dyn_hair_long_bk_l_01, swing1LimitDegrees=5, swing2LimitDegrees=30, twistLimitDegrees=2}
 {bone1Name=dyn_hair_long_bk_l_03, bone2Name=dyn_hair_long_bk_l_02, swing1LimitDegrees=5, swing2LimitDegrees=45, twistLimitDegrees=2}
 {bone1Name=dyn_hair_long_bk_l_04, bone2Name=dyn_hair_long_bk_l_03, swing1LimitDegrees=15, swing2LimitDegrees=30, twistLimitDegrees=2}
 {bone1Name=dyn_hair_long_bk_l_05, bone2Name=dyn_hair_long_bk_l_04, swing1LimitDegrees=30, swing2LimitDegrees=30, twistLimitDegrees=2}
 {bone1Name=dyn_hair_long_bk_lb_04, bone2Name=dyn_hair_long_bk_l_03, swing1LimitDegrees=15, swing2LimitDegrees=35, twistLimitDegrees=2}
 {bone1Name=dyn_hair_long_bk_lb_05, bone2Name=dyn_hair_long_bk_lb_04, swing1LimitDegrees=30, swing2LimitDegrees=30, twistLimitDegrees=2}
 {bone1Name=dyn_hair_long_bk_02, bone2Name=dyn_hair_long_bk_01, swing1LimitDegrees=5, swing2LimitDegrees=35, twistLimitDegrees=2}
 {bone1Name=dyn_hair_long_bk_03, bone2Name=dyn_hair_long_bk_02, swing1LimitDegrees=5, swing2LimitDegrees=45, twistLimitDegrees=2}
 {bone1Name=dyn_hair_long_bk_04, bone2Name=dyn_hair_long_bk_03, swing1LimitDegrees=5, swing2LimitDegrees=45, twistLimitDegrees=2}
 {bone1Name=dyn_hair_long_bk_05, bone2Name=dyn_hair_long_bk_04, swing1LimitDegrees=5, swing2LimitDegrees=30, twistLimitDegrees=2}
 {bone1Name=dyn_hair_long_bk_r_02, bone2Name=neck_02, swing1LimitDegrees=2, swing2LimitDegrees=45, twistLimitDegrees=2}
 {bone1Name=dyn_hair_long_bk_r_03, bone2Name=dyn_hair_long_bk_r_02, swing1LimitDegrees=10, swing2LimitDegrees=45, twistLimitDegrees=2}
 {bone1Name=dyn_hair_long_bk_r_04, bone2Name=dyn_hair_long_bk_r_03, swing1LimitDegrees=15, swing2LimitDegrees=30, twistLimitDegrees=2}
 {bone1Name=dyn_hair_long_bk_r_05, bone2Name=dyn_hair_long_bk_r_04, swing1LimitDegrees=30, swing2LimitDegrees=30, twistLimitDegrees=2}
 {bone1Name=dyn_hair_long_bk_rb_04, bone2Name=dyn_hair_long_bk_r_03, swing1LimitDegrees=15, swing2LimitDegrees=35, twistLimitDegrees=2}
 {bone1Name=dyn_hair_long_bk_rb_05, bone2Name=dyn_hair_long_bk_rb_04, swing1LimitDegrees=30, swing2LimitDegrees=30, twistLimitDegrees=2}
 {
  bone1Name=spine_05
  bone2Name=spine_04
  swing1LimitDegrees=45
  swing1Motion=Free
  swing2LimitDegrees=45
  swing2Motion=Free
  twistLimitDegrees=45
  twistMotion=Free
 }
 {
  bone1Name=clavicle_l
  bone2Name=spine_05
  swing1LimitDegrees=45
  swing1Motion=Free
  swing2LimitDegrees=45
  swing2Motion=Free
  twistLimitDegrees=45
  twistMotion=Free
 }
 {
  bone1Name=clavicle_r
  bone2Name=spine_05
  swing1LimitDegrees=45
  swing1Motion=Free
  swing2LimitDegrees=45
  swing2Motion=Free
  twistLimitDegrees=45
  twistMotion=Free
 }
 {
  bone1Name=spine_04
  bone2Name=spine_03
  swing1LimitDegrees=45
  swing1Motion=Free
  swing2LimitDegrees=45
  swing2Motion=Free
  twistLimitDegrees=45
  twistMotion=Free
 }
 {
  bone1Name=spine_03
  bone2Name=spine_02
  swing1LimitDegrees=45
  swing1Motion=Free
  swing2LimitDegrees=45
  swing2Motion=Free
  twistLimitDegrees=45
  twistMotion=Free
 }
 {
  bone1Name=spine_02
  bone2Name=spine_01
  swing1LimitDegrees=45
  swing1Motion=Free
  swing2LimitDegrees=45
  swing2Motion=Free
  twistLimitDegrees=45
  twistMotion=Free
 }
 {
  bone1Name=neck_02
  bone2Name=spine_05
  swing1LimitDegrees=45
  swing1Motion=Free
  swing2LimitDegrees=45
  swing2Motion=Free
  twistLimitDegrees=45
  twistMotion=Free
 }
 {
  bone1Name=dyn_hair_long_bk_l_01
  bone2Name=neck_02
  swing1LimitDegrees=45
  swing1Motion=Free
  swing2LimitDegrees=45
  swing2Motion=Free
  twistLimitDegrees=45
  twistMotion=Free
 }
 {
  bone1Name=upperarm_l
  bone2Name=clavicle_l
  swing1LimitDegrees=45
  swing1Motion=Free
  swing2LimitDegrees=45
  swing2Motion=Free
  twistLimitDegrees=45
  twistMotion=Free
 }
 {
  bone1Name=upperarm_r
  bone2Name=clavicle_r
  swing1LimitDegrees=45
  swing1Motion=Free
  swing2LimitDegrees=45
  swing2Motion=Free
  twistLimitDegrees=45
  twistMotion=Free
 }
 {
  bone1Name=dyn_hair_long_bk_01
  bone2Name=neck_02
  swing1LimitDegrees=45
  swing1Motion=Free
  swing2LimitDegrees=45
  swing2Motion=Free
  twistLimitDegrees=45
  twistMotion=Free
 }
]
