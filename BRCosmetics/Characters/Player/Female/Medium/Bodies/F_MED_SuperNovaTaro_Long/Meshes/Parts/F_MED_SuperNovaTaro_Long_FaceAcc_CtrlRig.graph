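# Control Rig for the SuperNovaTaro_Long face accessory that applies pose-driven corrective offsets to long hair and crown bones.
# Spherical pose readers on the head and right upper arm store look, aim and arm-pose alphas, some multiplied into combined weights.
# A Crouching branch then selects weighted ModifyTransforms chains, with the crouched chain adding its own head pose reader for the crown.
# Several computed alphas, including look left/right and skydiving, drive no transforms.
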ControlRigBlueprint /BRCosmetics/Characters/Player/Female/Medium/Bodies/F_MED_SuperNovaTaro_Long/Meshes/Parts/F_MED_SuperNovaTaro_Long_FaceAcc_CtrlRig
graphs [
 ContainedGraph {nodes [
   n1 = RigVMFunctionEntryNode()
   SequenceExecution_12 = &RigVMFunction_Sequence(ExecuteContext=@n1.ExecuteContext)
   SequenceExecution_12_1 = &RigVMFunction_Sequence(ExecuteContext=@SequenceExecution_12.B)
   SequenceExecution_12_2 = &RigVMFunction_Sequence(ExecuteContext=@SequenceExecution_12_1.B)
   RigVMFunctionReturnNode(A=@SequenceExecution_12.A, B=@SequenceExecution_12_1.A, C=@SequenceExecution_12_2.A, D=@SequenceExecution_12_2.B)
  ]}
 ContainedGraph {nodes [
   n1 = RigVMFunctionEntryNode()
   SequenceExecution_4 = &RigVMFunction_Sequence(ExecuteContext=@n1.ExecuteContext)
   SequenceExecution_4_1 = &RigVMFunction_Sequence(ExecuteContext=@SequenceExecution_4.B)
   SequenceExecution_4_2 = &RigVMFunction_Sequence(ExecuteContext=@SequenceExecution_4_1.B)
   RigVMFunctionReturnNode(A=@SequenceExecution_4.A, B=@SequenceExecution_4_1.A, C=@SequenceExecution_4_2.A, D=@SequenceExecution_4_2.B)
  ]}
 ContainedGraph {nodes [
   n1 = RigVMFunctionEntryNode()
   SequenceExecution_3 = &RigVMFunction_Sequence(ExecuteContext=@n1.ExecuteContext)
   SequenceExecution_3_1 = &RigVMFunction_Sequence(ExecuteContext=@SequenceExecution_3.B)
   SequenceExecution_3_2 = &RigVMFunction_Sequence(ExecuteContext=@SequenceExecution_3_1.B)
   RigVMFunctionReturnNode(A=@SequenceExecution_3.A, B=@SequenceExecution_3_1.A, C=@SequenceExecution_3_2.A, D=@SequenceExecution_3_2.B)
  ]}
 ContainedGraph {nodes [
   n1 = RigVMFunctionEntryNode()
   SequenceExecution_2 = &RigVMFunction_Sequence(ExecuteContext=@n1.ExecuteContext)
   SequenceExecution_2_1 = &RigVMFunction_Sequence(ExecuteContext=@SequenceExecution_2.B)
   SequenceExecution_2_2 = &RigVMFunction_Sequence(ExecuteContext=@SequenceExecution_2_1.B)
   RigVMFunctionReturnNode(A=@SequenceExecution_2.A, B=@SequenceExecution_2_1.A, C=@SequenceExecution_2_2.A, D=@SequenceExecution_2_2.B)
  ]}
 ContainedGraph {nodes [
   n1 = RigVMFunctionEntryNode()
   SequenceExecution_7 = &RigVMFunction_Sequence(ExecuteContext=@n1.ExecuteContext)
   SequenceExecution_7_1 = &RigVMFunction_Sequence(ExecuteContext=@SequenceExecution_7.B)
   SequenceExecution_7_2 = &RigVMFunction_Sequence(ExecuteContext=@SequenceExecution_7_1.B)
   RigVMFunctionReturnNode(A=@SequenceExecution_7.A, B=@SequenceExecution_7_1.A, C=@SequenceExecution_7_2.A, D=@SequenceExecution_7_2.B)
  ]}
 RigVMFunctionLibrary {links [], nodes []}
 RigVMModel {nodes [
   RigUnit_BeginExecution = &RigUnit_BeginExecution()
   SequenceExecution_12 = RigVMAggregateNode(ExecuteContext=@RigUnit_BeginExecution.ExecutePin)
   RigVMFunction_ControlFlowBranch = &RigVMFunction_ControlFlowBranch(ExecuteContext=@SequenceExecution_12.B, Condition=RigVMVariableNode() {variable=Crouching}.Value)
   SequenceAggregate = &RigVMFunction_Sequence(ExecuteContext=@RigVMFunction_ControlFlowBranch.False)
   n1 = RigVMVariableNode() {variable=head_SPR_look_up_alpha}
   &RigUnit_ModifyTransforms(
    ItemToModify="((Item=(Type=Bone,Name=\"dyn_hair_long_bk_r_mstr\"),Transform=(Rotation=(X=0,Y=0,Z=0,W=1),Translation=(X=0,Y=0,Z=0),Scale3D=(X=1,Y=1,Z=1))),(Item=(Type=Bone,Name=\"dyn_hair_long_bk_l_mstr\"),Transform=(Rotation=(X=0,Y=0,Z=0,W=1),Translation=(X=0,Y=0,Z=0),Scale ...(truncated)"
    ItemToModify.0.Transform.Rotation=&RigVMFunction_MathQuaternionFromEuler(Euler=(X=0,Y=0,Z=0)).Result
    ItemToModify.1.Transform.Rotation=&RigVMFunction_MathQuaternionFromEuler(Euler=(X=0,Y=0,Z=0)).Result
    ItemToModify.2.Transform.Rotation=&RigVMFunction_MathQuaternionFromEuler(Euler=(X=0,Y=0,Z=0)).Result
    ExecutePin=@SequenceAggregate.A
    Weight=@n1.Value
   )
   > &RigUnit_ModifyTransforms(
    ItemToModify=((Item=(Type=Bone,Name="dyn_crown_long_mstr"),Transform=(Rotation=(X=0,Y=0,Z=0,W=1),Translation=(X=-1,Y=0,Z=0),Scale3D=(X=1,Y=1,Z=1))))
    ItemToModify.0.Transform.Rotation=&RigVMFunction_MathQuaternionFromEuler(Euler=(X=0,Y=-5,Z=0)).Result
    Weight=@n1.Value
   )
   > &RigUnit_ModifyTransforms(
    ItemToModify=((Item=(Type=Bone,Name="dyn_hair_long_bk_mstr"),Transform=(Rotation=(X=0,Y=0,Z=0,W=1),Translation=(X=0,Y=0,Z=2),Scale3D=(X=1,Y=1,Z=1))))
    ItemToModify.0.Transform.Rotation=&RigVMFunction_MathQuaternionFromEuler(Euler=(X=0,Y=10,Z=0)).Result
    Weight=@n1.Value
   )
   > &RigUnit_ModifyTransforms(
    ItemToModify=((Item=(Type=Bone,Name="dyn_long_bk_dfrm"),Transform=(Rotation=(X=0,Y=0,Z=0,W=1),Translation=(X=5,Y=-5,Z=0),Scale3D=(X=1,Y=1,Z=1))))
    Weight=@n1.Value
   )
   &RigUnit_ModifyTransforms(
    ItemToModify="((Item=(Type=Bone,Name=\"dyn_hair_long_bk_mstr\"),Transform=(Rotation=(X=0,Y=0,Z=0,W=1),Translation=(X=0,Y=0,Z=0),Scale3D=(X=1,Y=1,Z=1))),(Item=(Type=Bone,Name=\"dyn_hair_long_bk_r_02\"),Transform=(Rotation=(X=0,Y=0,Z=0,W=1),Translation=(X=0,Y=0,Z=0),Scale3D=( ...(truncated)"
    ItemToModify.0.Transform.Rotation=&RigVMFunction_MathQuaternionFromEuler(Euler=(X=10,Y=0,Z=0)).Result
    ItemToModify.1.Transform.Rotation=&RigVMFunction_MathQuaternionFromEuler(Euler=(X=0,Y=5,Z=0)).Result
    Weight=RigVMVariableNode() {variable=uppearm_r_SPR_aim_up_two_hands_alpha}.Value
    ExecutePin=@SequenceAggregate.B
   )
   n2 = RigVMVariableNode() {variable=look_up_two_hands_rifle_multiplied_alpha}
   > &RigUnit_ModifyTransforms(
    ItemToModify=((Item=(Type=Bone,Name="dyn_hair_long_bk_mstr"),Transform=(Rotation=(X=0,Y=0,Z=0,W=1),Translation=(X=0,Y=0,Z=0),Scale3D=(X=1,Y=1,Z=1))))
    ItemToModify.0.Transform.Rotation=&RigVMFunction_MathQuaternionFromEuler(Euler=(X=0,Y=0,Z=0)).Result
    Weight=@n2.Value
   )
   > &RigUnit_ModifyTransforms(
    ItemToModify=((Item=(Type=Bone,Name="dyn_crown_long_mstr"),Transform=(Rotation=(X=0,Y=0,Z=0,W=1),Translation=(X=1,Y=0,Z=0),Scale3D=(X=1,Y=1,Z=1))),(Item=(Type=Bone,Name="None"),Transform=(Rotation=(X=0,Y=0,Z=0,W=1),Translation=(X=0,Y=0,Z=0),Scale3D=(X=1,Y=1,Z=1))))
    ItemToModify.0.Transform.Rotation=&RigVMFunction_MathQuaternionFromEuler(Euler=(X=0,Y=-5,Z=0)).Result
    Weight=@n2.Value
   )
   > &RigUnit_ModifyTransforms(
    ItemToModify=((Item=(Type=Bone,Name="dyn_hair_long_side_r_mstr"),Transform=(Rotation=(X=0,Y=0,Z=0,W=1),Translation=(X=2,Y=2,Z=0),Scale3D=(X=1,Y=1,Z=1))))
    Weight=RigVMVariableNode() {variable=head_SPR_look_down_alpha}.Value
   )
   n3 = RigVMVariableNode() {variable=head_SPR_look_up_alpha}
   &RigUnit_ModifyTransforms(
    ItemToModify=((Item=(Type=Bone,Name="dyn_crown_long_mstr"),Transform=(Rotation=(X=0,Y=0,Z=0,W=1),Translation=(X=-1,Y=0,Z=0),Scale3D=(X=1,Y=1,Z=1))))
    ItemToModify.0.Transform.Rotation=&RigVMFunction_MathQuaternionFromEuler(Euler=(X=0,Y=-12,Z=0)).Result
    ExecutePin=@RigVMFunction_ControlFlowBranch.True
    Weight=@n3.Value
   )
   > &RigUnit_ModifyTransforms(
    ItemToModify=((Item=(Type=Bone,Name="dyn_long_bk_dfrm"),Transform=(Rotation=(X=0,Y=0,Z=0,W=1),Translation=(X=6,Y=-6,Z=0),Scale3D=(X=1,Y=1,Z=1))))
    Weight=@n3.Value
   )
   n4 = RigVMVariableNode() {variable=uppearm_r_SPR_aim_up_two_hands_alpha}
   > &RigUnit_ModifyTransforms(
    ItemToModify=((Item=(Type=Bone,Name="dyn_crown_long_mstr"),Transform=(Rotation=(X=0,Y=0,Z=0,W=1),Translation=(X=-1,Y=0,Z=0),Scale3D=(X=1,Y=1,Z=1))))
    Weight=@n4.Value
   )
   > &RigUnit_ModifyTransforms(
    ItemToModify="((Item=(Type=Bone,Name=\"dyn_hair_long_bk_l_mstr\"),Transform=(Rotation=(X=0,Y=0,Z=0,W=1),Translation=(X=-1,Y=-1,Z=0),Scale3D=(X=1,Y=1,Z=1))),(Item=(Type=Bone,Name=\"dyn_hair_long_bk_l_02\"),Transform=(Rotation=(X=0,Y=0,Z=0,W=1),Translation=(X=0,Y=0,Z=-1),Scal ...(truncated)"
    Weight=@n4.Value
   )
   > &RigUnit_ModifyTransforms(
    ItemToModify=((Item=(Type=Bone,Name="dyn_hair_long_fr_l_02"),Transform=(Rotation=(X=0,Y=0,Z=0,W=1),Translation=(X=1,Y=2,Z=0),Scale3D=(X=1,Y=1,Z=1))))
    Weight=@n4.Value
   )
   > n5 = &RigUnit_SphericalPoseReader(
    ActiveRegionScaleFactors=(PositiveWidth=2,NegativeWidth=0.5,PositiveHeight=2,NegativeHeight=2)
    ActiveRegionSize=0.075
    Debug=(bDrawDebug=true,DebugScale=25,DebugSegments=20,DebugThickness=0.25)
    DriverAxis=(X=1,Y=0,Z=0)
    DriverItem=(Type=Bone,Name="head")
    FalloffRegionScaleFactors=(PositiveWidth=1,NegativeWidth=0.4,PositiveHeight=1,NegativeHeight=1)
    FalloffSize=0.25
    FlipHeightScaling=false
    FlipWidthScaling=true
    OptionalParentItem=(Type=Bone,Name="spine_05")
    RotationOffset=(X=-70,Y=5,Z=0)
   )
   > &RigUnit_ModifyTransforms(
    ItemToModify=((Item=(Type=Bone,Name="dyn_crown_long_mstr"),Transform=(Rotation=(X=0,Y=0,Z=0,W=1),Translation=(X=-2.5,Y=0,Z=0),Scale3D=(X=1,Y=1,Z=1))))
    ItemToModify.0.Transform.Rotation=&RigVMFunction_MathQuaternionFromEuler(Euler=(X=0,Y=-22,Z=0)).Result
    Weight=@n5.OutputParam
   )
   n6 = RigVMVariableNode() {variable=look_up_two_hands_rifle_multiplied_alpha}
   > &RigUnit_ModifyTransforms(
    ItemToModify="((Item=(Type=Bone,Name=\"dyn_hair_long_bk_l_mstr\"),Transform=(Rotation=(X=0,Y=0,Z=0,W=1),Translation=(X=-1,Y=-1,Z=0),Scale3D=(X=1,Y=1,Z=1))),(Item=(Type=Bone,Name=\"dyn_hair_long_bk_l_02\"),Transform=(Rotation=(X=0,Y=0,Z=0,W=1),Translation=(X=0,Y=0,Z=-1),Scal ...(truncated)"
    ItemToModify.0.Transform.Rotation=&RigVMFunction_MathQuaternionFromEuler(Euler=(X=0,Y=0,Z=0)).Result
    ItemToModify.1.Transform.Rotation=&RigVMFunction_MathQuaternionFromEuler(Euler=(X=0,Y=0,Z=0)).Result
    Weight=@n6.Value
   )
   > &RigUnit_ModifyTransforms(
    ItemToModify=((Item=(Type=Bone,Name="dyn_hair_long_fr_l_02"),Transform=(Rotation=(X=0,Y=0,Z=0,W=1),Translation=(X=1,Y=2.5,Z=0),Scale3D=(X=1,Y=1,Z=1))))
    Weight=@n6.Value
   )
   SequenceExecution_3 = RigVMAggregateNode(ExecuteContext=@SequenceExecution_12.A)
   SequenceExecution_4 = RigVMAggregateNode(ExecuteContext=@SequenceExecution_3.A)
   SequenceExecution_7 = RigVMAggregateNode(ExecuteContext=@SequenceExecution_4.A)
   n7 = &RigUnit_SphericalPoseReader(
    ActiveRegionScaleFactors=(PositiveWidth=1,NegativeWidth=1,PositiveHeight=1,NegativeHeight=1)
    ActiveRegionSize=0.065
    Debug=(bDrawDebug=false,DebugScale=25,DebugSegments=20,DebugThickness=0.25)
    DriverAxis=(X=-1,Y=0,Z=0)
    DriverItem=(Type=Bone,Name="upperarm_r")
    FalloffRegionScaleFactors=(PositiveWidth=1,NegativeWidth=1,PositiveHeight=1,NegativeHeight=1)
    FalloffSize=0.125
    OptionalParentItem=(Type=Bone)
    RotationOffset=(X=60,Y=0,Z=65)
    ExecutePin=@SequenceExecution_7.A
   )
   > RigVMVariableNode(Value=@n7.OutputParam) {variable=upperarm_r_SPR_back_alpha}
   n8 = &RigUnit_SphericalPoseReader(
    ActiveRegionScaleFactors=(PositiveWidth=1,NegativeWidth=1,PositiveHeight=1,NegativeHeight=1)
    ActiveRegionSize=0.075
    Debug=(bDrawDebug=false,DebugScale=25,DebugSegments=20,DebugThickness=0.25)
    DriverAxis=(X=-1,Y=0,Z=0)
    DriverItem=(Type=Bone,Name="upperarm_r")
    FalloffRegionScaleFactors=(PositiveWidth=1,NegativeWidth=1,PositiveHeight=1,NegativeHeight=1)
    FalloffSize=0.2
    OptionalParentItem=(Type=Bone)
    RotationOffset=(X=-55,Y=0,Z=-20)
    ExecutePin=@SequenceExecution_7.D
   )
   > RigVMVariableNode(Value=@n8.OutputParam) {variable=upperarm_r_rifle_up_alpha}
   n9 = &RigUnit_SphericalPoseReader(
    ActiveRegionScaleFactors=(PositiveWidth=1,NegativeWidth=1,PositiveHeight=1,NegativeHeight=1)
    ActiveRegionSize=0.03
    Debug=(bDrawDebug=false,DebugScale=25,DebugSegments=20,DebugThickness=0.25)
    DriverAxis=(X=-1,Y=0,Z=0)
    DriverItem=(Type=Bone,Name="upperarm_r")
    FalloffRegionScaleFactors=(PositiveWidth=1,NegativeWidth=1,PositiveHeight=1,NegativeHeight=1)
    FalloffSize=0.125
    OptionalParentItem=(Type=Bone)
    RotationOffset=(X=-115,Y=0,Z=25)
    ExecutePin=@SequenceExecution_7.B
   )
   > RigVMVariableNode(Value=@n9.OutputParam) {variable=uppearm_r_SPR_aim_up_two_hands_alpha}
   n10 = &RigUnit_SphericalPoseReader(
    ActiveRegionScaleFactors=(PositiveWidth=2,NegativeWidth=2,PositiveHeight=1,NegativeHeight=1)
    ActiveRegionSize=0.05
    Debug=(bDrawDebug=false,DebugScale=25,DebugSegments=20,DebugThickness=0.25)
    DriverAxis=(X=-1,Y=0,Z=0)
    DriverItem=(Type=Bone,Name="upperarm_r")
    FalloffRegionScaleFactors=(PositiveWidth=1,NegativeWidth=1,PositiveHeight=0.75,NegativeHeight=0.75)
    FalloffSize=0.2
    FlipWidthScaling=false
    OptionalParentItem=(Type=Bone)
    RotationOffset=(X=15,Y=150,Z=0)
    ExecutePin=@SequenceExecution_7.C
   )
   > RigVMVariableNode(Value=@n10.OutputParam) {variable=upperarm_r_skydiving}
   SequenceExecution_2 = RigVMAggregateNode(ExecuteContext=@SequenceExecution_4.B)
   n11 = &RigUnit_SphericalPoseReader(
    ActiveRegionScaleFactors=(PositiveWidth=1,NegativeWidth=0.5,PositiveHeight=1.5,NegativeHeight=1.5)
    ActiveRegionSize=0.1
    Debug=(bDrawDebug=true,DebugScale=25,DebugSegments=20,DebugThickness=0.25)
    DriverAxis=(X=0,Y=1,Z=0)
    DriverItem=(Type=Bone,Name="head")
    FalloffRegionScaleFactors=(PositiveWidth=1,NegativeWidth=0.5,PositiveHeight=1.5,NegativeHeight=1.5)
    FalloffSize=0.2
    FlipWidthScaling=false
    OptionalParentItem=(Type=Bone)
    RotationOffset=(X=80,Y=0,Z=-25)
    ExecutePin=@SequenceExecution_2.A
   )
   > RigVMVariableNode(Value=@n11.OutputParam) {variable=head_SPR_look_up_alpha}
   n12 = &RigUnit_SphericalPoseReader(
    ActiveRegionScaleFactors=(PositiveWidth=2,NegativeWidth=0.5,PositiveHeight=2,NegativeHeight=2)
    ActiveRegionSize=0.075
    Debug=(bDrawDebug=false,DebugScale=25,DebugSegments=20,DebugThickness=0.25)
    DriverAxis=(X=0,Y=1,Z=0)
    DriverItem=(Type=Bone,Name="head")
    FalloffRegionScaleFactors=(PositiveWidth=1,NegativeWidth=0.4,PositiveHeight=1,NegativeHeight=1)
    FalloffSize=0.2
    FlipHeightScaling=false
    FlipWidthScaling=true
    OptionalParentItem=(Type=Bone)
    RotationOffset=(X=90,Y=0,Z=12.5)
    ExecutePin=@SequenceExecution_2.B
   )
   > RigVMVariableNode(Value=@n12.OutputParam) {variable=head_SPR_look_down_alpha}
   n13 = &RigUnit_SphericalPoseReader(
    ActiveRegionScaleFactors=(PositiveWidth=0.75,NegativeWidth=0.75,PositiveHeight=1,NegativeHeight=0.5)
    ActiveRegionSize=0.1
    Debug=(bDrawDebug=false,DebugScale=25,DebugSegments=20,DebugThickness=0.25)
    DriverAxis=(X=0,Y=1,Z=0)
    DriverItem=(Type=Bone,Name="head")
    FalloffRegionScaleFactors=(PositiveWidth=1,NegativeWidth=1,PositiveHeight=1,NegativeHeight=0.5)
    FalloffSize=0.2
    OptionalParentItem=(Type=Bone)
    RotationOffset=(X=120,Y=0,Z=0)
    ExecutePin=@SequenceExecution_2.C
   )
   > RigVMVariableNode(Value=@n13.OutputParam) {variable=head_SPR_look_left_alpha}
   n14 = &RigUnit_SphericalPoseReader(
    ActiveRegionScaleFactors=(PositiveWidth=0.75,NegativeWidth=0.75,PositiveHeight=1,NegativeHeight=0.5)
    ActiveRegionSize=0.1
    Debug=(bDrawDebug=false,DebugScale=25,DebugSegments=20,DebugThickness=0.25)
    DriverAxis=(X=0,Y=1,Z=0)
    DriverItem=(Type=Bone,Name="head")
    FalloffRegionScaleFactors=(PositiveWidth=1,NegativeWidth=1,PositiveHeight=1,NegativeHeight=0.5)
    FalloffSize=0.2
    FlipHeightScaling=true
    FlipWidthScaling=false
    OptionalParentItem=(Type=Bone)
    RotationOffset=(X=60,Y=0,Z=0)
    ExecutePin=@SequenceExecution_2.D
   )
   > RigVMVariableNode(Value=@n14.OutputParam) {variable=head_SPR_look_right_alpha}
   RigVMVariableNode(
    Value=&RigVMFunction_MathFloatMul(B=RigVMVariableNode() {variable=uppearm_r_SPR_aim_up_two_hands_alpha}.Value, A=RigVMVariableNode() {variable=head_SPR_look_up_alpha}.Value) {template="Multiply::Execute(in A,in B,out Result)"}.Result
    ExecuteContext=@SequenceExecution_3.B
   ) {variable=look_up_and_left_two_hands_multiplied_alpha}
   RigVMVariableNode(
    Value=&RigVMFunction_MathFloatMul(A=RigVMVariableNode() {variable=head_SPR_look_up_alpha}.Value, B=RigVMVariableNode() {variable=upperarm_r_rifle_up_alpha}.Value) {template="Multiply::Execute(in A,in B,out Result)"}.Result
    ExecuteContext=@SequenceExecution_3.C
   ) {variable=look_up_two_hands_rifle_multiplied_alpha}
   RigVMVariableNode(
    Value=&RigVMFunction_MathFloatMul(B=RigVMVariableNode() {variable=head_SPR_look_down_alpha}.Value, A=RigVMVariableNode() {variable=upperarm_r_SPR_back_alpha}.Value) {template="Multiply::Execute(in A,in B,out Result)"}.Result
    ExecuteContext=@SequenceExecution_3.D
   ) {variable=look_down_upperarm_r_back_multiplied_alpha}
  ]}
]
hierarchy [
 root {loc=<0 0 0>, rot=<0 0 0 1>}
 pelvis {parent=root, loc=<0 -0.5301 98.0571>, rot=<0.0127 0.707 -0.0127 -0.707>}
 spine_01 {parent=pelvis, loc=<6.1208 -0.9328 0>, rot=<0 0 0.034 -0.9994>}
 spine_02 {parent=spine_01, loc=<6.5996 -0.2095 0>, rot=<0 0 0.0238 -0.9997>}
 spine_03 {parent=spine_02, loc=<6.4914 -0.1237 0>, rot=<0 0 0.0149 -0.9999>}
 spine_04 {parent=spine_03, loc=<6.9598 0.0794 0>, rot=<0 0 -0.0132 -0.9999>}
 spine_05 {parent=spine_04, loc=<6.9987 0 0>, rot=<0 0 -0.0015 -1>}
 neck_01 {parent=spine_05, loc=<11.3138 0.1529 0>, rot=<0 0 -0.2003 -0.9797>}
 neck_02 {parent=neck_01, loc=<4.5627 0 0>, rot=<0 0 0.0504 -0.9987>}
 head {parent=neck_02, loc=<5.3914 0 0>, rot=<0 0 0.115 -0.9934>}
 dfrm_hair_long_mstr {parent=head, loc=<9.6892 -1.5079 0>, rot=<0.5021 -0.4979 -0.5021 -0.4979>}
 dyn_hair_long_fr_r_mstr {parent=dfrm_hair_long_mstr, loc=<-3.8928 3.3852 9.3393>, rot=<-0.5028 -0.4972 0.5869 0.3944>}
 dyn_hair_long_fr_r_01 {parent=dyn_hair_long_fr_r_mstr, loc=<0 0 0>, rot=<0 0 0 1>}
 dyn_hair_long_fr_r_02 {parent=dyn_hair_long_fr_r_01, loc=<10.2364 0 0>, rot=<-0.0078 -0.0746 0.0948 -0.9927>}
 dyn_hair_long_fr_r_03 {parent=dyn_hair_long_fr_r_02, loc=<9.1073 0 0>, rot=<0.0112 -0.0936 0.1274 -0.9874>}
 dyn_hair_long_fr_r_04 {parent=dyn_hair_long_fr_r_03, loc=<10.9595 0 0>, rot=<0.0472 -0.1584 0.0413 -0.9854>}
 dyn_hair_long_fr_rb_03 {parent=dyn_hair_long_fr_r_02, loc=<3.596 -1.7077 -1.2055>, rot=<-0.0044 -0.0793 0.0554 -0.9953>}
 dyn_hair_long_side_r_mstr {parent=dfrm_hair_long_mstr, loc=<-6.6765 6.9449 2.8608>, rot=<-0.507 -0.4929 0.5517 0.4423>}
 dyn_hair_long_side_r_01 {parent=dyn_hair_long_side_r_mstr, loc=<0 0 0>, rot=<0 0 0 1>}
 dyn_hair_long_side_r_02 {parent=dyn_hair_long_side_r_01, loc=<10.0576 0 0>, rot=<-0.004 0.0395 0.2236 -0.9739>}
 dyn_hair_long_side_r_03 {parent=dyn_hair_long_side_r_02, loc=<7.8017 0 0>, rot=<0.0192 -0.0508 0.0337 -0.998>}
 dyn_hair_long_fr_l_mstr {parent=dfrm_hair_long_mstr, loc=<6.8863 6.9165 5.3912>, rot=<-0.4665 -0.5314 -0.4641 -0.5335>}
 dyn_hair_long_fr_l_01 {parent=dyn_hair_long_fr_l_mstr, loc=<0 0 0>, rot=<0 0 0 1>}
 dyn_hair_long_fr_l_02 {parent=dyn_hair_long_fr_l_01, loc=<9.6016 0 0>, rot=<0 0.0005 -0.1393 -0.9903>}
 dyn_hair_long_fr_l_03 {parent=dyn_hair_long_fr_l_02, loc=<7.9325 0.251 -0.0384>, rot=<-0.0045 -0.0175 -0.1092 -0.9939>}
 dyn_hair_long_bk_r_mstr {parent=dfrm_hair_long_mstr, loc=<-5.7981 10.7266 -3.3934>, rot=<-0.6837 -0.7217 0.0784 -0.0743>}
 dyn_hair_long_bk_r_01 {parent=dyn_hair_long_bk_r_mstr, loc=<0 0 0>, rot=<0 0 0 1>}
 dyn_hair_long_bk_r_02 {parent=dyn_hair_long_bk_r_01, loc=<7.4902 0.0543 -0.2205>, rot=<0 0.1325 0 -0.9912>}
 dyn_hair_long_bk_r_03 {parent=dyn_hair_long_bk_r_02, loc=<6.8109 0 -0.0003>, rot=<0 -0.0674 0 -0.9977>}
 dyn_hair_long_bk_rb_04 {parent=dyn_hair_long_bk_r_03, loc=<5.345 -2.5199 0.5349>, rot=<0.0062 -0.1346 0.0453 -0.9898>}
 dyn_hair_long_bk_rb_05 {parent=dyn_hair_long_bk_rb_04, loc=<6.4743 -0.2015 1.2143>, rot=<-0.2637 0.0113 -0.0432 -0.9636>}
 dyn_hair_long_bk_r_04 {parent=dyn_hair_long_bk_r_03, loc=<7.1599 0.4174 0.7028>, rot=<-0.0036 -0.1101 -0.0323 -0.9934>}
 dyn_hair_long_bk_r_05 {parent=dyn_hair_long_bk_r_04, loc=<7.9714 0 -0.0001>, rot=<0 -0.0753 0 -0.9972>}
 dyn_hair_long_bk_l_mstr {parent=dfrm_hair_long_mstr, loc=<5.7981 12.6344 -3.3934>, rot=<0.7217 0.6837 -0.0743 0.0784>}
 dyn_hair_long_bk_l_01 {parent=dyn_hair_long_bk_l_mstr, loc=<0 0 0>, rot=<0 0 0 1>}
 dyn_hair_long_bk_l_02 {parent=dyn_hair_long_bk_l_01, loc=<5.6299 0.0465 0.189>, rot=<0 0.1325 0 -0.9912>}
 dyn_hair_long_bk_l_03 {parent=dyn_hair_long_bk_l_02, loc=<6.8104 0 0>, rot=<0 -0.0674 0 -0.9977>}
 dyn_hair_long_bk_lb_04 {parent=dyn_hair_long_bk_l_03, loc=<5.1458 2.6264 -0.0327>, rot=<-0.0062 -0.1346 -0.0453 -0.9898>}
 dyn_hair_long_bk_lb_05 {parent=dyn_hair_long_bk_lb_04, loc=<6.5864 0.2016 0.0742>, rot=<0.2665 -0.0725 0.0201 -0.9609>}
 dyn_hair_long_bk_l_04 {parent=dyn_hair_long_bk_l_03, loc=<7.1603 -0.4174 0.7026>, rot=<0.0036 -0.1101 0.0323 -0.9934>}
 dyn_hair_long_bk_l_05 {parent=dyn_hair_long_bk_l_04, loc=<7.9712 0 0>, rot=<0 -0.0753 0 -0.9972>}
 dyn_hair_long_bk_mstr {parent=dfrm_hair_long_mstr, loc=<0 11.909 -4.9615>, rot=<-0.703 -0.703 0.0764 -0.0764>}
 dyn_hair_long_bk_01 {parent=dyn_hair_long_bk_mstr, loc=<0 0 0>, rot=<0 0 0 1>}
 dyn_hair_long_bk_02 {parent=dyn_hair_long_bk_01, loc=<6.3371 0 0.0333>, rot=<0 0.1325 0 -0.9912>}
 dyn_hair_long_bk_03 {parent=dyn_hair_long_bk_02, loc=<6.8104 0 0>, rot=<0 -0.0674 0 -0.9977>}
 dyn_hair_long_bk_04 {parent=dyn_hair_long_bk_03, loc=<9.1171 0 0>, rot=<0 -0.1102 0 -0.9939>}
 dyn_hair_long_bk_05 {parent=dyn_hair_long_bk_04, loc=<9.6786 0 -0.2173>, rot=<0 -0.0753 0 -0.9972>}
 dyn_long_bk_dfrm {parent=dfrm_hair_long_mstr, loc=<0.3549 -0.0345 -9.2631>, rot=<-0.0374 -0.7061 -0.0532 -0.7051>}
 dyn_crown_long_mstr {parent=head, loc=<2.5881 -0.7095 0>, rot=<0.0029 -0.7071 0.7071 -0.0029>}
 dyn_crown_long {parent=dyn_crown_long_mstr, loc=<0 0 0>, rot=<0 0 0 1>}
 dyn_crown_long_fx {parent=dyn_crown_long_mstr, loc=<0 0 0>, rot=<0 0 0 1>}
 dfrm_hair_short_mstr {parent=head, loc=<9.6892 -1.5079 0>, rot=<0.5021 -0.4979 -0.5021 -0.4979>}
 dyn_hair_short_bangs_r_mstr {parent=dfrm_hair_short_mstr, loc=<-4.837 2.6943 8.9341>, rot=<0.1069 0.1069 -0.699 -0.699>}
 dyn_hair_short_bangs_r_01 {parent=dyn_hair_short_bangs_r_mstr, loc=<0 0 0>, rot=<0 0 0 1>}
 dyn_hair_short_bangs_l_mstr {parent=dfrm_hair_short_mstr, loc=<3.9682 3.7549 9.9609>, rot=<-0.1183 -0.1183 -0.6971 -0.6971>}
 dyn_hair_short_bangs_l_01 {parent=dyn_hair_short_bangs_l_mstr, loc=<0 0 0>, rot=<0 0 0 1>}
 dyn_hair_short_bangs_c_mstr {parent=dfrm_hair_short_mstr, loc=<-0.2554 2.521 10.8217>, rot=<-0.027 -0.027 -0.7066 -0.7066>}
 dyn_hair_short_bangs_c_01 {parent=dyn_hair_short_bangs_c_mstr, loc=<0 0 0>, rot=<0 0 0 1>}
 dyn_hair_short_side_r_mstr {parent=dfrm_hair_short_mstr, loc=<-6.815 8.6093 -2.4006>, rot=<0.5953 0.5953 -0.3816 -0.3816>}
 dyn_hair_short_side_r_01 {parent=dyn_hair_short_side_r_mstr, loc=<0 0 0>, rot=<0 0 0 1>}
 dyn_hair_short_side_r_02 {parent=dyn_hair_short_side_r_01, loc=<6.6085 0 0>, rot=<0 0 0 1>}
 dyn_hair_short_fr_r_mstr {parent=dfrm_hair_short_mstr, loc=<-7.0478 4.0589 4.4053>, rot=<0.5 0.5 -0.5 -0.5>}
 dyn_hair_short_fr_r_01 {parent=dyn_hair_short_fr_r_mstr, loc=<0 0 0>, rot=<0 0 0 1>}
 dyn_hair_short_fr_r_02 {parent=dyn_hair_short_fr_r_01, loc=<9.4825 0 0>, rot=<0 0 0 1>}
 dyn_hair_short_bk_r_mstr {parent=dfrm_hair_short_mstr, loc=<-2.2928 11.6719 -7.418>, rot=<0.7027 0.7027 -0.079 -0.079>}
 dyn_hair_short_bk_r_01 {parent=dyn_hair_short_bk_r_mstr, loc=<0 0 0>, rot=<0 0 0 1>}
 dyn_hair_short_bk_r_02 {parent=dyn_hair_short_bk_r_01, loc=<4.5196 0 0>, rot=<0 0 0 1>}
 dyn_hair_short_bk_l_mstr {parent=dfrm_hair_short_mstr, loc=<5.4946 12.6902 -5.3369>, rot=<0.6565 0.6565 0.2627 0.2627>}
 dyn_hair_short_bk_l_01 {parent=dyn_hair_short_bk_l_mstr, loc=<0 0 0>, rot=<0 0 0 1>}
 dyn_hair_short_bk_l_02 {parent=dyn_hair_short_bk_l_01, loc=<3.5303 0 0>, rot=<0 0 0 1>}
 dyn_hair_short_side_l_mstr {parent=dfrm_hair_short_mstr, loc=<7.7084 9.4883 0.1982>, rot=<-0.5 -0.5 -0.5 -0.5>}
 dyn_hair_short_side_l_01 {parent=dyn_hair_short_side_l_mstr, loc=<0 0 0>, rot=<0 0 0 1>}
 dyn_hair_short_side_l_02 {parent=dyn_hair_short_side_l_01, loc=<6.6435 0 0>, rot=<0 0 0 1>}
 dyn_hair_short_fr_l_mstr {parent=dfrm_hair_short_mstr, loc=<7.0478 7.2378 4.4053>, rot=<-0.5 -0.5 -0.5 -0.5>}
 dyn_hair_short_fr_l_01 {parent=dyn_hair_short_fr_l_mstr, loc=<0 0 0>, rot=<0 0 0 1>}
 dyn_hair_short_fr_l_02 {parent=dyn_hair_short_fr_l_01, loc=<7.4064 0 0>, rot=<0 0 0 1>}
 clavicle_l {parent=spine_05, loc=<5.4831 -0.1053 -3.2946>, rot=<0.0465 -0.7289 -0.009 -0.6829>}
 upperarm_l {parent=clavicle_l, loc=<9.8658 -1.2136 -0.4105>, rot=<0.0164 -0.3245 -0.0269 -0.9454>}
 upperarm_twist_01_l {parent=upperarm_l, loc=<7.2135 -0.1134 0>, rot=<0 0 0 1>}
 clavicle_r {parent=spine_05, loc=<5.4831 -0.1053 3.2946>, rot=<0.7289 0.0465 -0.6829 0.009>}
 upperarm_r {parent=clavicle_r, loc=<-9.8658 1.2136 0.4104>, rot=<0.0164 -0.3245 -0.0269 -0.9454>}
 upperarm_twist_01_r {parent=upperarm_r, loc=<-7.2135 0.1134 0>, rot=<0 0 0 -1>}
]
variables [
 Crouching {type=bool, default=False}
 Moving {type=double, default=0}
 clavicle_r_SPR_back_alpha {type=double, default=0}
 head_SPR_look_down_alpha {type=double, default=0}
 head_SPR_look_left_alpha {type=double, default=0}
 head_SPR_look_right_alpha {type=double, default=0}
 head_SPR_look_up_alpha {type=double, default=0}
 look_down_upperarm_r_back_multiplied_alpha {type=double, default=0}
 look_up_and_left_two_hands_multiplied_alpha {type=double, default=0}
 look_up_two_hands_rifle_multiplied_alpha {type=double, default=0}
 uppearm_r_SPR_aim_up_two_hands_alpha {type=double, default=0}
 upperarm_r_SPR_back_alpha {type=double, default=0}
 upperarm_r_rifle_up_alpha {type=double, default=0}
 upperarm_r_skydiving {type=double, default=0}
]
