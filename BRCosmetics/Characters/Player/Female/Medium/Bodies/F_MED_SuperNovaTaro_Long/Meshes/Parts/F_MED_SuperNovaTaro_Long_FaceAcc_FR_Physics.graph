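# Physics asset for the front face-accessory part of the long-haired female SuperNovaTaro cosmetic body, providing physics bodies for its dangling hair strands.
# Capsules on the upper spine, neck, head, clavicles and upper arms form a body proxy joined by free-motion constraints.
# Front-left, front-right and right-side hair chains hang from free master bodies on the head, with per-joint swing and tight twist limits.
# An extra front-right branch capsule splits off the second segment.
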PhysicsAsset /BRCosmetics/Characters/Player/Female/Medium/Bodies/F_MED_SuperNovaTaro_Long/Meshes/Parts/F_MED_SuperNovaTaro_Long_FaceAcc_FR_Physics
bodies {
 clavicle_l [{shapeType=Capsule, length=5.3544, radius=5.3769, center=<6.4482 -0.3686 -2.4355>, rotation=<rot 85.606 0 0>}]
 clavicle_r [{shapeType=Capsule, length=4.8637, radius=5.8676, center=<-6.4482 0.3686 2.4355>, rotation=<rot 85.606 0 0>}]
 dyn_hair_long_fr_l_01 [{shapeType=Capsule, length=5.5, radius=2.5, shapeName=dyn_hair_long_fr_l_01_capsule, center=<2.4362 0 2>, rotation=<rot 90 0 0>}]
 dyn_hair_long_fr_l_02 [{shapeType=Capsule, length=5.5, radius=2.5, shapeName=dyn_hair_long_fr_l_02_capsule, center=<2.4362 0.75 0>, rotation=<rot 90 0 0>}]
 dyn_hair_long_fr_l_03 [{shapeType=Capsule, length=8, radius=2, shapeName=dyn_hair_long_fr_l_03_capsule, center=<4.1155 1 0>, rotation=<rot 90 0 0>}]
 dyn_hair_long_fr_l_mstr [{shapeType=Capsule, length=1, radius=1, shapeName=dyn_hair_long_fr_l_mstr_capsule}]
 dyn_hair_long_fr_r_01 [{shapeType=Capsule, length=6, radius=2, shapeName=dyn_hair_long_fr_r_01_capsule, center=<2.5 0 2>, rotation=<rot 90 0 0>}]
 dyn_hair_long_fr_r_02 [{shapeType=Capsule, length=6.5, radius=1.5, shapeName=dyn_hair_long_fr_r_02_capsule, center=<2.5 0 0.1767>, rotation=<rot 90 0 0>}]
 dyn_hair_long_fr_r_03 [{shapeType=Capsule, length=7, radius=1.5, shapeName=dyn_hair_long_fr_r_03_capsule, center=<3.25 0.75 0>, rotation=<rot 90 0 0>}]
 dyn_hair_long_fr_r_04 [{shapeType=Capsule, length=9, radius=1.5, shapeName=dyn_hair_long_fr_r_04_capsule, center=<2.25 0 0>, rotation=<rot 80 -180 -179.9999>}]
 dyn_hair_long_fr_r_mstr [{shapeType=Capsule, length=1, radius=1, shapeName=dyn_hair_long_fr_r_mstr_capsule}]
 dyn_hair_long_fr_rb_03 [{
   shapeType=Capsule
   length=3.5785
   radius=0.6826
   shapeName=dyn_hair_long_fr_rb_03_capsule
   center=<5.3439 0.0302 -0.9402>
   rotation=<rot 70.0769 -5.0771 5.0766>
  }]
 dyn_hair_long_side_r_01 [{shapeType=Capsule, length=5.5, radius=1.5, shapeName=dyn_hair_long_side_r_01_capsule, center=<3.1756 0 0.8501>, rotation=<rot 85 0 0>}]
 dyn_hair_long_side_r_02 [{shapeType=Capsule, length=7, radius=1.5, shapeName=dyn_hair_long_side_r_02_capsule, center=<1.5 0 0.0553>, rotation=<rot 90 0 0>}]
 dyn_hair_long_side_r_03 [{shapeType=Capsule, length=7.5, radius=1, shapeName=dyn_hair_long_side_r_03_capsule, center=<3.1692 0 0>, rotation=<rot 90 0 0>}]
 dyn_hair_long_side_r_mstr [{shapeType=Capsule, length=1, radius=1, shapeName=dyn_hair_long_side_r_mstr_capsule}]
 head [{shapeType=Capsule, length=5.5, radius=6.5, center=<2.1606 1.409 0>, rotation=<rot 90 0 0>}]
 neck_01 [{shapeType=Capsule, length=6.5, radius=5.5, shapeName=neck_01_capsule, rotation=<rot 90 0 0>}]
 spine_04 [{shapeType=Capsule, length=9.7013, radius=7.968, center=<0.6532 1.8714 0>}]
 spine_05 [{shapeType=Capsule, length=15.0318, radius=7.5064, center=<2.4257 -1.9164 -0.0001>}]
 upperarm_l [{shapeType=Capsule, length=23.0812, radius=4.1347, center=<10.8056 -0.4936 0.3456>, rotation=<rot 90 0 0>}]
 upperarm_r [{shapeType=Capsule, length=24.0812, radius=3.1347, center=<-10.8056 0.4936 -0.3456>, rotation=<rot 90 0 0>}]
}
constraints [
 {
  bone1Name=head
  bone2Name=neck_01
  swing1LimitDegrees=45
  swing1Motion=Free
  swing2LimitDegrees=45
  swing2Motion=Free
  twistLimitDegrees=45
  twistMotion=Free
 }
 {
  bone1Name=spine_05
  bone2Name=spine_04
  swing1LimitDegrees=45
  swing1Motion=Free
  swing2LimitDegrees=45
  swing2Motion=Free
  twistLimitDegrees=45
  twistMotion=Free
 }
 {
  bone1Name=clavicle_l
  bone2Name=spine_05
  swing1LimitDegrees=45
  swing1Motion=Free
  swing2LimitDegrees=45
  swing2Motion=Free
  twistLimitDegrees=45
  twistMotion=Free
 }
 {
  bone1Name=upperarm_l
  bone2Name=clavicle_l
  swing1LimitDegrees=45
  swing1Motion=Free
  swing2LimitDegrees=45
  swing2Motion=Free
  twistLimitDegrees=45
  twistMotion=Free
 }
 {
  bone1Name=clavicle_r
  bone2Name=spine_05
  swing1LimitDegrees=45
  swing1Motion=Free
  swing2LimitDegrees=45
  swing2Motion=Free
  twistLimitDegrees=45
  twistMotion=Free
 }
 {
  bone1Name=upperarm_r
  bone2Name=clavicle_r
  swing1LimitDegrees=45
  swing1Motion=Free
  swing2LimitDegrees=45
  swing2Motion=Free
  twistLimitDegrees=45
  twistMotion=Free
 }
 {
  bone1Name=dyn_hair_long_fr_l_mstr
  bone2Name=head
  swing1LimitDegrees=45
  swing1Motion=Free
  swing2LimitDegrees=45
  swing2Motion=Free
  twistLimitDegrees=5
  twistMotion=Free
 }
 {bone1Name=dyn_hair_long_fr_l_03, bone2Name=dyn_hair_long_fr_l_02, swing1LimitDegrees=35, swing2LimitDegrees=35, twistLimitDegrees=5}
 {bone1Name=dyn_hair_long_fr_l_02, bone2Name=dyn_hair_long_fr_l_01, swing1LimitDegrees=25, swing2LimitDegrees=25, twistLimitDegrees=5}
 {bone1Name=dyn_hair_long_fr_l_01, bone2Name=dyn_hair_long_fr_l_mstr, swing1LimitDegrees=20, swing2LimitDegrees=10, twistLimitDegrees=5}
 {
  bone1Name=dyn_hair_long_fr_r_mstr
  bone2Name=head
  swing1LimitDegrees=45
  swing1Motion=Free
  swing2LimitDegrees=45
  swing2Motion=Free
  twistLimitDegrees=5
  twistMotion=Free
 }
 {bone1Name=dyn_hair_long_fr_r_01, bone2Name=dyn_hair_long_fr_r_mstr, swing1LimitDegrees=20, swing2LimitDegrees=20, twistLimitDegrees=5}
 {bone1Name=dyn_hair_long_fr_r_02, bone2Name=dyn_hair_long_fr_r_01, swing1LimitDegrees=20, swing2LimitDegrees=20, twistLimitDegrees=5}
 {bone1Name=dyn_hair_long_fr_r_03, bone2Name=dyn_hair_long_fr_r_02, swing1LimitDegrees=30, swing2LimitDegrees=30, twistLimitDegrees=5}
 {bone1Name=dyn_hair_long_fr_r_04, bone2Name=dyn_hair_long_fr_r_03, swing1LimitDegrees=35, swing2LimitDegrees=35, twistLimitDegrees=5}
 {
  bone1Name=dyn_hair_long_side_r_mstr
  bone2Name=head
  swing1LimitDegrees=45
  swing1Motion=Free
  swing2LimitDegrees=45
  swing2Motion=Free
  twistLimitDegrees=5
  twistMotion=Free
 }
 {bone1Name=dyn_hair_long_side_r_01, bone2Name=dyn_hair_long_side_r_mstr, swing1LimitDegrees=35, swing2LimitDegrees=25, twistLimitDegrees=5}
 {bone1Name=dyn_hair_long_side_r_02, bone2Name=dyn_hair_long_side_r_01, swing1LimitDegrees=20, swing2LimitDegrees=20, twistLimitDegrees=5}
 {bone1Name=dyn_hair_long_side_r_03, bone2Name=dyn_hair_long_side_r_02, swing1LimitDegrees=25, swing2LimitDegrees=25, twistLimitDegrees=5}
 {
  bone1Name=neck_01
  bone2Name=spine_05
  swing1LimitDegrees=45
  swing1Motion=Free
  swing2LimitDegrees=45
  swing2Motion=Free
  twistLimitDegrees=45
  twistMotion=Free
 }
 {bone1Name=dyn_hair_long_fr_rb_03, bone2Name=dyn_hair_long_fr_r_02, swing1LimitDegrees=30, swing2LimitDegrees=30, twistLimitDegrees=2}
]
