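# Physics asset for the face-accessory part of the SuperNovaTaro Long female medium cosmetic body, rigging long back hair for dynamic motion.
# Center, left and right capsule chains hang from master bones on the head, with extra side branches splitting off the third left and right segments.
# Chain roots are nearly locked, while lower joints allow limited swing and minimal twist.
# Head, neck, spine, clavicle and upper-arm capsules form the body shapes, linked by free-rotation constraints.
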PhysicsAsset /BRCosmetics/Characters/Player/Female/Medium/Bodies/F_MED_SuperNovaTaro_Long/Meshes/Parts/F_MED_SuperNovaTaro_Long_FaceAcc_Physics
bodies {
 clavicle_l [{shapeType=Capsule, length=5.3544, radius=5.3769, center=<6.4482 -0.111 -2.4355>, rotation=<rot 85.606 0 0>}]
 clavicle_r [{shapeType=Capsule, length=4.8637, radius=5.8676, center=<-6.4482 0.3686 2.4355>, rotation=<rot 85.606 0 0>}]
 dyn_hair_long_bk_01 [{shapeType=Capsule, length=4.5143, radius=3.2691, shapeName=dyn_hair_long_bk_01_capsule, center=<3.19 0 1>, rotation=<rot 90 0 0>}]
 dyn_hair_long_bk_02 [{shapeType=Capsule, length=4.039, radius=4.4433, shapeName=dyn_hair_long_bk_02_capsule, center=<3 0 2.3>, rotation=<rot 90 0 0>}]
 dyn_hair_long_bk_03 [{shapeType=Capsule, length=2.9405, radius=5.2157, shapeName=dyn_hair_long_bk_03_capsule, center=<3 0 2.3>, rotation=<rot 90 0 0>}]
 dyn_hair_long_bk_04 [{shapeType=Capsule, length=3.8626, radius=5.283, shapeName=dyn_hair_long_bk_04_capsule, center=<3 0 2.3>, rotation=<rot 90 0 0>}]
 dyn_hair_long_bk_05 [{shapeType=Capsule, length=4, radius=5, shapeName=dyn_hair_long_bk_05_capsule, center=<3 0 0>, rotation=<rot 90 0 0>}]
 dyn_hair_long_bk_l_01 [{shapeType=Capsule, length=4.7059, radius=3.7336, shapeName=dyn_hair_long_bk_l_01_capsule, center=<3.19 0 1>, rotation=<rot 90 0 0>}]
 dyn_hair_long_bk_l_02 [{shapeType=Capsule, length=2.9899, radius=3.5865, shapeName=dyn_hair_long_bk_l_02_capsule, center=<3 0 3.5>, rotation=<rot 90 0 0>}]
 dyn_hair_long_bk_l_03 [{shapeType=Capsule, length=3.9597, radius=4.6257, shapeName=dyn_hair_long_bk_l_03_capsule, center=<3 0 3.5>, rotation=<rot 90 0 0>}]
 dyn_hair_long_bk_l_04 [{shapeType=Capsule, length=2.239, radius=4.621, shapeName=dyn_hair_long_bk_l_04_capsule, center=<3 0 3.5>, rotation=<rot 90 0 0>}]
 dyn_hair_long_bk_l_05 [{shapeType=Capsule, length=3.971, radius=5.0825, shapeName=dyn_hair_long_bk_l_05_capsule, center=<3 0 0.75>, rotation=<rot 90 0 0>}]
 dyn_hair_long_bk_l_mstr [{shapeType=Capsule, length=0.505, radius=0.505, shapeName=dyn_hair_long_bk_l_mstr_capsule, rotation=<rot 90 0 0>}]
 dyn_hair_long_bk_lb_04 [{shapeType=Capsule, length=1.7842, radius=3.8474, shapeName=dyn_hair_long_bk_lb_04_capsule, center=<3 0 3.5>, rotation=<rot 90 0 0>}]
 dyn_hair_long_bk_lb_05 [{shapeType=Capsule, length=2.8626, radius=1.5197, shapeName=dyn_hair_long_bk_lb_05_capsule, center=<3 0 0>, rotation=<rot 90 0 0>}]
 dyn_hair_long_bk_mstr [{shapeType=Capsule, length=0.505, radius=0.505, shapeName=dyn_hair_long_bk_mstr_capsule, rotation=<rot 90 0 0>}]
 dyn_hair_long_bk_r_01 [{shapeType=Capsule, length=5.2146, radius=3.7812, shapeName=dyn_hair_long_bk_r_01_capsule, center=<3.19 0 1>, rotation=<rot 90 0 0>}]
 dyn_hair_long_bk_r_02 [{shapeType=Capsule, length=3.101, radius=3.9777, shapeName=dyn_hair_long_bk_r_02_capsule, center=<3 0 3.5>, rotation=<rot 90 0 0>}]
 dyn_hair_long_bk_r_03 [{shapeType=Capsule, length=4.2685, radius=4.7321, shapeName=dyn_hair_long_bk_r_03_capsule, center=<3 0 3.5>, rotation=<rot 90 0 0>}]
 dyn_hair_long_bk_r_04 [{shapeType=Capsule, length=1.3384, radius=4.5703, shapeName=dyn_hair_long_bk_r_04_capsule, center=<3 0 3.5>, rotation=<rot 90 0 0>}]
 dyn_hair_long_bk_r_05 [{shapeType=Capsule, length=6.0655, radius=4.3041, shapeName=dyn_hair_long_bk_r_05_capsule, center=<3 0 0.75>, rotation=<rot 90 0 0>}]
 dyn_hair_long_bk_r_mstr [{shapeType=Capsule, length=0.505, radius=0.505, shapeName=dyn_hair_long_bk_r_mstr_capsule, rotation=<rot 90 0 0>}]
 dyn_hair_long_bk_rb_04 [{shapeType=Capsule, length=2.5602, radius=4.055, shapeName=dyn_hair_long_bk_rb_04_capsule, center=<3 0 3.5>, rotation=<rot 90 0 0>}]
 dyn_hair_long_bk_rb_05 [{shapeType=Capsule, length=4.9851, radius=1.4305, shapeName=dyn_hair_long_bk_rb_05_capsule, center=<3 0 0>, rotation=<rot 90 0 0>}]
 head [{shapeType=Capsule, length=8.2132, radius=7.5167, shapeName=head_capsule, rotation=<rot 90 0 0>}]
 neck_02 [{shapeType=Capsule, length=2.5, radius=4.5, shapeName=neck_02_capsule, center=<-4.0488 0 0>, rotation=<rot 90 0 0>}]
 spine_01 [{shapeType=Capsule, length=7.0613, radius=9.6791, center=<0 1.0381 -0.0001>}]
 spine_02 [{shapeType=Capsule, length=5.9705, radius=8.255, center=<0 1.813 -0.0001>}]
 spine_03 [{shapeType=Capsule, length=7.9414, radius=7.7882, center=<0 2.0457 -0.0001>}]
 spine_04 [{shapeType=Capsule, length=9.7013, radius=7.968, center=<0.6532 1.8714 0>}]
 spine_05 [{shapeType=Capsule, length=15.0318, radius=7.5064, center=<2.25 0 0>}]
 upperarm_l [{shapeType=Capsule, length=18.0812, radius=4.1347, center=<10.8056 -0.4936 0.3456>, rotation=<rot 90 0 0>}]
 upperarm_r [{shapeType=Capsule, length=18.0812, radius=4.1347, center=<-10.8056 0.4936 -0.3456>, rotation=<rot 90 0 0>}]
}
constraints [
 {bone1Name=dyn_hair_long_bk_l_02, bone2Name=dyn_hair_long_bk_l_01, swing1LimitDegrees=30, swing2LimitDegrees=30, twistLimitDegrees=2}
 {bone1Name=dyn_hair_long_bk_l_03, bone2Name=dyn_hair_long_bk_l_02, swing1LimitDegrees=30, swing2LimitDegrees=30, twistLimitDegrees=2}
 {bone1Name=dyn_hair_long_bk_l_04, bone2Name=dyn_hair_long_bk_l_03, swing1LimitDegrees=20, swing2LimitDegrees=30, twistLimitDegrees=2}
 {bone1Name=dyn_hair_long_bk_l_05, bone2Name=dyn_hair_long_bk_l_04, swing1LimitDegrees=30, swing2LimitDegrees=40, twistLimitDegrees=2}
 {bone1Name=dyn_hair_long_bk_lb_04, bone2Name=dyn_hair_long_bk_l_03, swing1LimitDegrees=10, swing2LimitDegrees=20, twistLimitDegrees=2}
 {bone1Name=dyn_hair_long_bk_lb_05, bone2Name=dyn_hair_long_bk_lb_04, swing1LimitDegrees=10, swing2LimitDegrees=20, twistLimitDegrees=2}
 {bone1Name=dyn_hair_long_bk_02, bone2Name=dyn_hair_long_bk_01, swing1LimitDegrees=30, swing2LimitDegrees=30, twistLimitDegrees=2}
 {bone1Name=dyn_hair_long_bk_03, bone2Name=dyn_hair_long_bk_02, swing1LimitDegrees=30, swing2LimitDegrees=30, twistLimitDegrees=2}
 {bone1Name=dyn_hair_long_bk_04, bone2Name=dyn_hair_long_bk_03, swing1LimitDegrees=20, swing2LimitDegrees=30, twistLimitDegrees=2}
 {bone1Name=dyn_hair_long_bk_05, bone2Name=dyn_hair_long_bk_04, swing1LimitDegrees=35, swing2LimitDegrees=45, twistLimitDegrees=2}
 {bone1Name=dyn_hair_long_bk_r_02, bone2Name=dyn_hair_long_bk_r_01, swing1LimitDegrees=30, swing2LimitDegrees=30, twistLimitDegrees=2}
 {bone1Name=dyn_hair_long_bk_r_03, bone2Name=dyn_hair_long_bk_r_02, swing1LimitDegrees=30, swing2LimitDegrees=30, twistLimitDegrees=2}
 {bone1Name=dyn_hair_long_bk_r_04, bone2Name=dyn_hair_long_bk_r_03, swing1LimitDegrees=20, swing2LimitDegrees=30, twistLimitDegrees=2}
 {bone1Name=dyn_hair_long_bk_r_05, bone2Name=dyn_hair_long_bk_r_04, swing1LimitDegrees=30, swing2LimitDegrees=40, twistLimitDegrees=2}
 {bone1Name=dyn_hair_long_bk_rb_04, bone2Name=dyn_hair_long_bk_r_03, swing1LimitDegrees=20, swing2LimitDegrees=30, twistLimitDegrees=2}
 {bone1Name=dyn_hair_long_bk_rb_05, bone2Name=dyn_hair_long_bk_rb_04, swing1LimitDegrees=30, swing2LimitDegrees=40, twistLimitDegrees=2}
 {
  bone1Name=spine_05
  bone2Name=spine_04
  swing1LimitDegrees=45
  swing1Motion=Free
  swing2LimitDegrees=45
  swing2Motion=Free
  twistLimitDegrees=45
  twistMotion=Free
 }
 {
  bone1Name=clavicle_l
  bone2Name=spine_05
  swing1LimitDegrees=45
  swing1Motion=Free
  swing2LimitDegrees=45
  swing2Motion=Free
  twistLimitDegrees=45
  twistMotion=Free
 }
 {
  bone1Name=upperarm_l
  bone2Name=clavicle_l
  swing1LimitDegrees=45
  swing1Motion=Free
  swing2LimitDegrees=45
  swing2Motion=Free
  twistLimitDegrees=45
  twistMotion=Free
 }
 {
  bone1Name=clavicle_r
  bone2Name=spine_05
  swing1LimitDegrees=45
  swing1Motion=Free
  swing2LimitDegrees=45
  swing2Motion=Free
  twistLimitDegrees=45
  twistMotion=Free
 }
 {
  bone1Name=upperarm_r
  bone2Name=clavicle_r
  swing1LimitDegrees=45
  swing1Motion=Free
  swing2LimitDegrees=45
  swing2Motion=Free
  twistLimitDegrees=45
  twistMotion=Free
 }
 {
  bone1Name=spine_04
  bone2Name=spine_03
  swing1LimitDegrees=45
  swing1Motion=Free
  swing2LimitDegrees=45
  swing2Motion=Free
  twistLimitDegrees=45
  twistMotion=Free
 }
 {
  bone1Name=spine_03
  bone2Name=spine_02
  swing1LimitDegrees=45
  swing1Motion=Free
  swing2LimitDegrees=45
  swing2Motion=Free
  twistLimitDegrees=45
  twistMotion=Free
 }
 {
  bone1Name=spine_02
  bone2Name=spine_01
  swing1LimitDegrees=45
  swing1Motion=Free
  swing2LimitDegrees=45
  swing2Motion=Free
  twistLimitDegrees=45
  twistMotion=Free
 }
 {
  bone1Name=dyn_hair_long_bk_mstr
  bone2Name=head
  swing1LimitDegrees=45
  swing1Motion=Free
  swing2LimitDegrees=45
  swing2Motion=Free
  twistLimitDegrees=45
  twistMotion=Free
 }
 {bone1Name=dyn_hair_long_bk_01, bone2Name=dyn_hair_long_bk_mstr, swing1LimitDegrees=1, swing2LimitDegrees=1, twistLimitDegrees=1}
 {
  bone1Name=dyn_hair_long_bk_r_mstr
  bone2Name=head
  swing1LimitDegrees=45
  swing1Motion=Free
  swing2LimitDegrees=45
  swing2Motion=Free
  twistLimitDegrees=45
  twistMotion=Free
 }
 {bone1Name=dyn_hair_long_bk_r_01, bone2Name=dyn_hair_long_bk_r_mstr, swing1LimitDegrees=1, swing2LimitDegrees=1, twistLimitDegrees=1}
 {
  bone1Name=dyn_hair_long_bk_l_mstr
  bone2Name=head
  swing1LimitDegrees=45
  swing1Motion=Free
  swing2LimitDegrees=45
  swing2Motion=Free
  twistLimitDegrees=45
  twistMotion=Free
 }
 {bone1Name=dyn_hair_long_bk_l_01, bone2Name=dyn_hair_long_bk_l_mstr, swing1LimitDegrees=1, swing2LimitDegrees=1, twistLimitDegrees=1}
 {
  bone1Name=head
  bone2Name=neck_02
  swing1LimitDegrees=45
  swing1Motion=Free
  swing2LimitDegrees=45
  swing2Motion=Free
  twistLimitDegrees=45
  twistMotion=Free
 }
 {
  bone1Name=neck_02
  bone2Name=spine_05
  swing1LimitDegrees=45
  swing1Motion=Free
  swing2LimitDegrees=45
  swing2Motion=Free
  twistLimitDegrees=45
  twistMotion=Free
 }
]
