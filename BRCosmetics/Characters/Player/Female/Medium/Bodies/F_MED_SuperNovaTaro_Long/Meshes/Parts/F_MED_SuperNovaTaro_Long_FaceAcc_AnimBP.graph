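# Animation Blueprint for the F_MED_SuperNovaTaro_Long FaceAcc part, driving the character's long hair.
# Crouching and default branches copy body poses, constrain hair bones to head, spine and clavicles, then run stacked RigidBody simulations.
# A bool blend selects the branch, feeding a Control Rig driven by crouch and movement.
# Updates set gravity override; back bling zeroes velocity and acceleration scales and enables an extra default-branch simulation.
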AnimBlueprint /BRCosmetics/Characters/Player/Female/Medium/Bodies/F_MED_SuperNovaTaro_Long/Meshes/Parts/F_MED_SuperNovaTaro_Long_FaceAcc_AnimBP
target_skeleton=&/F_MED_SuperNovaTaro_Long_FaceAcc_Skeleton
graphs [
 AnimGraph {nodes [
   "Constraint - dyn_hair_long_bk_05 (P)"(Alpha=1, ConstraintWeights_0=1) {
    alphaBoolBlend {blendOption=Linear}
    boneToModify {boneName=dyn_hair_long_bk_05}
    constraintSetup [{targetBone {boneName=dyn_hair_long_bk_04}}]
    constraintWeights [1]
   }
   n1 = "Constraint - dyn_hair_long_fr_l_02 (P)"(Alpha=SelectFloat(A=0.5, B=0, bPickA=Get bIsSkyDiving()), ConstraintWeights_0=1) {
    alphaBoolBlend {blendOption=Linear}
    boneToModify {boneName=dyn_hair_long_fr_l_02}
    constraintSetup [{targetBone {boneName=clavicle_l}}]
    constraintWeights [1]
   }
   "Constraint - dyn_hair_long_side_r_01 (P)"(ComponentPose=@n1, Alpha=SelectFloat(A=1, B=0, bPickA=Get bIsSkyDiving()), ConstraintWeights_0=1) {
    alphaBoolBlend {blendOption=Linear}
    boneToModify {boneName=dyn_hair_long_side_r_01}
    constraintSetup [{targetBone {boneName=spine_05}}]
    constraintWeights [1]
   }
   "Constraint - dyn_hair_long_bk_03 (P)"(Alpha=1, ConstraintWeights_0=1) {
    alphaBoolBlend {blendOption=Linear}
    boneToModify {boneName=dyn_hair_long_bk_03}
    constraintSetup [{targetBone {boneName=spine_05}}]
    constraintWeights [1]
   }
   "Constraint - dyn_hair_long_bk_04 (P)"(Alpha=1, ConstraintWeights_0=1) {
    alphaBoolBlend {blendOption=Linear}
    boneToModify {boneName=dyn_hair_long_bk_04}
    constraintSetup [{targetBone {boneName=spine_03}}]
    constraintWeights [1]
   }
   n2 = Copy Pose From Mesh(SourceMeshComponent=Get Skel_Mesh()) {bUseAttachedParent=true}
   n3 = Copy Pose From Mesh() {bUseAttachedParent=true}
   n4 = "Constraint - dyn_hair_long_bk_mstr (PP)"!(
    ComponentPose=Local To Component(LocalPose=Layered blend per bone(BasePose=@n3, BlendPoses_0=@n2) {layerSetup [{branchFilters [{boneName=upperarm_l}, {boneName=upperarm_r}]}]})
    Alpha=1
    ConstraintWeights_0=0.3
    ConstraintWeights_1:float=0.7
   ) -> (Pose) {
    alphaBoolBlend {blendOption=Linear}
    boneToModify {boneName=dyn_hair_long_bk_mstr}
    constraintSetup [{targetBone {boneName=head}}, {targetBone {boneName=spine_05}}]
    constraintWeights [1,1]
   }
   n5 = "Constraint - dyn_hair_long_bk_01 (PP)"!(ComponentPose=@n4, Alpha=1, ConstraintWeights_0=0, ConstraintWeights_1:float=1) -> (Pose) {
    alphaBoolBlend {blendOption=Linear}
    boneToModify {boneName=dyn_hair_long_bk_01}
    constraintSetup [{targetBone {boneName=head}}, {targetBone {boneName=spine_05}}]
    constraintWeights [1,1]
   }
   n6 = "Constraint - dyn_hair_long_bk_02 (PP)"!(ComponentPose=@n5, Alpha=1, ConstraintWeights_0=0.3, ConstraintWeights_1:float=0.7) -> (Pose) {
    alphaBoolBlend {blendOption=Linear}
    boneToModify {boneName=dyn_hair_long_bk_02}
    constraintSetup [{targetBone {boneName=head}}, {targetBone {boneName=spine_04}}]
    constraintWeights [1,1]
   }
   n7 = "Constraint - dyn_hair_long_bk_l_mstr (PP)"!(ComponentPose=@n6, Alpha=1, ConstraintWeights_0=0.3, ConstraintWeights_1:float=0.7) -> (Pose) {
    alphaBoolBlend {blendOption=Linear}
    boneToModify {boneName=dyn_hair_long_bk_l_mstr}
    constraintSetup [{targetBone {boneName=spine_05}}, {targetBone {boneName=clavicle_l}}]
    constraintWeights [1,1]
   }
   n8 = "Constraint - dyn_hair_long_bk_l_01 (PP)"!(ComponentPose=@n7, Alpha=1, ConstraintWeights_0=0.2, ConstraintWeights_1:float=0.8) -> (Pose) {
    alphaBoolBlend {blendOption=Linear}
    boneToModify {boneName=dyn_hair_long_bk_l_01}
    constraintSetup [{targetBone {boneName=head}}, {targetBone {boneName=dyn_hair_long_bk_01}}]
    constraintWeights [1,1]
   }
   n9 = "Constraint - dyn_hair_long_bk_l_02 (PP)"!(ComponentPose=@n8, Alpha=1, ConstraintWeights_0=0.3, ConstraintWeights_1:float=0.7) -> (Pose) {
    alphaBoolBlend {blendOption=Linear}
    boneToModify {boneName=dyn_hair_long_bk_l_02}
    constraintSetup [{targetBone {boneName=spine_05}}, {targetBone {boneName=clavicle_l}}]
    constraintWeights [1,1]
   }
   n10 = "Constraint - dyn_hair_long_bk_r_mstr (PP)"!(ComponentPose=@n9, Alpha=1, ConstraintWeights_0=0.3, ConstraintWeights_1:float=0.7) -> (Pose) {
    alphaBoolBlend {blendOption=Linear}
    boneToModify {boneName=dyn_hair_long_bk_r_mstr}
    constraintSetup [{targetBone {boneName=spine_05}}, {targetBone {boneName=clavicle_r}}]
    constraintWeights [1,1]
   }
   n11 = "Constraint - dyn_hair_long_bk_r_01 (PP)"!(ComponentPose=@n10, Alpha=1, ConstraintWeights_0=0.3, ConstraintWeights_1:float=0.7) -> (Pose) {
    alphaBoolBlend {blendOption=Linear}
    boneToModify {boneName=dyn_hair_long_bk_r_01}
    constraintSetup [{targetBone {boneName=head}}, {targetBone {boneName=dyn_hair_long_bk_01}}]
    constraintWeights [1,1]
   }
   n12 = "Constraint - dyn_hair_long_bk_r_02 (PP)"!(ComponentPose=@n11, Alpha=1, ConstraintWeights_0=0.3, ConstraintWeights_1:float=0.7) -> (Pose) {
    alphaBoolBlend {blendOption=Linear}
    boneToModify {boneName=dyn_hair_long_bk_r_02}
    constraintSetup [{targetBone {boneName=spine_05}}, {targetBone {boneName=clavicle_r}}]
    constraintWeights [1,1]
   }
   n13 = "Constraint - dyn_hair_long_side_r_01 (PP)"!(ComponentPose=@n12, Alpha=1, ConstraintWeights_0=0.4, ConstraintWeights_1:float=0.6) -> (Pose) {
    alphaBoolBlend {blendOption=Linear}
    boneToModify {boneName=dyn_hair_long_side_r_01}
    constraintSetup [{targetBone {boneName=spine_05}}, {targetBone {boneName=head}}]
    constraintWeights [1,1]
   }
   n14 = "Constraint - dyn_hair_long_fr_r_01 (PP)"!(ComponentPose=@n13, Alpha=1, ConstraintWeights_0=0.6, ConstraintWeights_1:float=0.4) -> (Pose) {
    alphaBoolBlend {blendOption=Linear}
    boneToModify {boneName=dyn_hair_long_fr_r_01}
    constraintSetup [{targetBone {boneName=spine_05}}, {targetBone {boneName=head}}]
    constraintWeights [1,1]
   }
   n15 = "Constraint - dyn_hair_long_fr_l_01 (PP)"!(ComponentPose=@n14, Alpha=1, ConstraintWeights_0=0.4, ConstraintWeights_1:float=0.6) -> (Pose) {
    alphaBoolBlend {blendOption=Linear}
    boneToModify {boneName=dyn_hair_long_fr_l_01}
    constraintSetup [{targetBone {boneName=spine_05}}, {targetBone {boneName=head}}]
    constraintWeights [1,1]
   }
   n16 = Get GravityOverride()
   n17 = RigidBody!(
    ComponentPose=@n15
    Alpha
    OverrideWorldGravity:Vector=@n16
    ExternalForce
    ComponentLinearAccScale:Vector=Get ACC() -> (ACC:Vector)
    ComponentLinearVelScale:Vector=Get VEL() -> (VEL:Vector)
   ) -> (Pose) {
    bClampLinearTranslationLimitToRefPose=true
    bOverrideWorldGravity=true
    overridePhysicsAsset=&/F_MED_SuperNovaTaro_Long_FaceAcc_Crouch
    simulationSpace=BaseBoneSpace
    alphaBoolBlend {blendOption=Linear}
    baseBoneRef {boneName=spine_05}
    componentAppliedLinearAccClamp=<2500 2500 2500>
    componentLinearAccScale=<0.5 0.5 0.5>
    componentLinearVelScale=<0.5 0.5 0.5>
    overrideWorldGravity {z=-980}
   }
   n18 = RigidBody!(ComponentPose=@n17, Alpha, OverrideWorldGravity=@n16, ExternalForce) -> (Pose) {
    bClampLinearTranslationLimitToRefPose=true
    bOverrideWorldGravity=true
    overridePhysicsAsset=&/F_MED_SuperNovaTaro_Long_FaceAcc_FR_Physics
    simulationSpace=BaseBoneSpace
    alphaBoolBlend {blendOption=Linear}
    baseBoneRef {boneName=spine_05}
    componentAppliedLinearAccClamp=<2500 2500 2500>
    componentLinearAccScale=<0.5 0.5 0.5>
    componentLinearVelScale=<0.5 0.5 0.5>
    overrideWorldGravity {z=-980}
   }
   n19 = Copy Pose From Mesh() {bUseAttachedParent=true}
   n20 = Copy Pose From Mesh(SourceMeshComponent=Get Skel_Mesh()) {bUseAttachedParent=true}
   n21 = "Constraint - dyn_hair_long_bk_mstr (PP)"(
    ComponentPose=Local To Component(LocalPose=Layered blend per bone(BasePose=@n19, BlendPoses_0=@n20) {layerSetup [{branchFilters [{boneName=upperarm_l}, {boneName=upperarm_r}]}]})
    Alpha=0.7
    ConstraintWeights_0=0.3
    ConstraintWeights_1=0.7
   ) {
    alphaBoolBlend {blendOption=Linear}
    boneToModify {boneName=dyn_hair_long_bk_mstr}
    constraintSetup [{targetBone {boneName=head}}, {targetBone {boneName=spine_05}}]
    constraintWeights [1,1]
   }
   n22 = "Constraint - dyn_hair_long_bk_01 (PP)"(ComponentPose=@n21, Alpha=1, ConstraintWeights_0=0.3, ConstraintWeights_1=0.7) {
    alphaBoolBlend {blendOption=Linear}
    boneToModify {boneName=dyn_hair_long_bk_01}
    constraintSetup [{targetBone {boneName=head}}, {targetBone {boneName=spine_05}}]
    constraintWeights [1,1]
   }
   n23 = "Constraint - dyn_hair_long_bk_02 (PP)"(ComponentPose=@n22, Alpha=1, ConstraintWeights_0=0, ConstraintWeights_1=1) {
    alphaBoolBlend {blendOption=Linear}
    boneToModify {boneName=dyn_hair_long_bk_02}
    constraintSetup [{targetBone {boneName=head}}, {targetBone {boneName=dyn_hair_long_bk_l_05}}]
    constraintWeights [1,1]
   }
   n24 = "Constraint - dyn_hair_long_bk_l_mstr (PP)"(ComponentPose=@n23, Alpha=0.7, ConstraintWeights_0=0.3, ConstraintWeights_1=0.7) {
    alphaBoolBlend {blendOption=Linear}
    boneToModify {boneName=dyn_hair_long_bk_l_mstr}
    constraintSetup [{targetBone {boneName=spine_05}}, {targetBone {boneName=clavicle_l}}]
    constraintWeights [1,1]
   }
   n25 = "Constraint - dyn_hair_long_bk_l_01 (PP)"(ComponentPose=@n24, Alpha=1, ConstraintWeights_0=0, ConstraintWeights_1=1) {
    alphaBoolBlend {blendOption=Linear}
    boneToModify {boneName=dyn_hair_long_bk_l_01}
    constraintSetup [{targetBone {boneName=head}}, {targetBone {boneName=dyn_hair_long_bk_01}}]
    constraintWeights [1,1]
   }
   n26 = "Constraint - dyn_hair_long_bk_l_02 (PP)"(ComponentPose=@n25, Alpha=1, ConstraintWeights_0=0.3, ConstraintWeights_1=0.7) {
    alphaBoolBlend {blendOption=Linear}
    boneToModify {boneName=dyn_hair_long_bk_l_02}
    constraintSetup [{targetBone {boneName=spine_05}}, {targetBone {boneName=clavicle_l}}]
    constraintWeights [1,1]
   }
   n27 = "Constraint - dyn_hair_long_bk_r_mstr (PP)"(ComponentPose=@n26, Alpha=0.7, ConstraintWeights_0=0.3, ConstraintWeights_1=0.7) {
    alphaBoolBlend {blendOption=Linear}
    boneToModify {boneName=dyn_hair_long_bk_r_mstr}
    constraintSetup [{targetBone {boneName=spine_05}}, {targetBone {boneName=clavicle_r}}]
    constraintWeights [1,1]
   }
   n28 = "Constraint - dyn_hair_long_bk_r_01 (PP)"(ComponentPose=@n27, Alpha=1, ConstraintWeights_0=0, ConstraintWeights_1=1) {
    alphaBoolBlend {blendOption=Linear}
    boneToModify {boneName=dyn_hair_long_bk_r_01}
    constraintSetup [{targetBone {boneName=head}}, {targetBone {boneName=dyn_hair_long_bk_01}}]
    constraintWeights [1,1]
   }
   n29 = "Constraint - dyn_hair_long_bk_r_02 (PP)"(ComponentPose=@n28, Alpha=1, ConstraintWeights_0=0.3, ConstraintWeights_1=0.7) {
    alphaBoolBlend {blendOption=Linear}
    boneToModify {boneName=dyn_hair_long_bk_r_02}
    constraintSetup [{targetBone {boneName=spine_05}}, {targetBone {boneName=clavicle_r}}]
    constraintWeights [1,1]
   }
   n30 = "Constraint - dyn_hair_long_side_r_01 (PP)"(ComponentPose=@n29, Alpha=1, ConstraintWeights_0=0.4, ConstraintWeights_1=0.6) {
    alphaBoolBlend {blendOption=Linear}
    boneToModify {boneName=dyn_hair_long_side_r_01}
    constraintSetup [{targetBone {boneName=spine_05}}, {targetBone {boneName=head}}]
    constraintWeights [1,1]
   }
   n31 = "Constraint - dyn_hair_long_fr_r_01 (PP)"(ComponentPose=@n30, Alpha=1, ConstraintWeights_0=0.6, ConstraintWeights_1=0.4) {
    alphaBoolBlend {blendOption=Linear}
    boneToModify {boneName=dyn_hair_long_fr_r_01}
    constraintSetup [{targetBone {boneName=spine_05}}, {targetBone {boneName=head}}]
    constraintWeights [1,1]
   }
   n32 = "Constraint - dyn_hair_long_fr_l_01 (PP)"(ComponentPose=@n31, Alpha=1, ConstraintWeights_0=0.4, ConstraintWeights_1=0.6) {
    alphaBoolBlend {blendOption=Linear}
    boneToModify {boneName=dyn_hair_long_fr_l_01}
    constraintSetup [{targetBone {boneName=spine_05}}, {targetBone {boneName=head}}]
    constraintWeights [1,1]
   }
   n33 = Get GravityOverride()
   n34 = RigidBody!(
    ComponentPose=@n32
    Alpha
    OverrideWorldGravity=@n33
    ExternalForce
    ComponentLinearAccScale:Vector=Get ACC()
    ComponentLinearVelScale:Vector=Get VEL()
   ) -> (Pose) {
    bClampLinearTranslationLimitToRefPose=true
    bOverrideWorldGravity=true
    overridePhysicsAsset=&/F_MED_SuperNovaTaro_Long_FaceAcc_Physics
    simulationSpace=BaseBoneSpace
    alphaBoolBlend {blendOption=Linear}
    baseBoneRef {boneName=spine_05}
    componentAppliedLinearAccClamp=<3500 3500 3500>
    componentLinearAccScale=<0.5 0.5 0.5>
    componentLinearVelScale=<0.5 0.5 0.5>
    overrideWorldGravity {z=-980}
   }
   n35 = RigidBody!(ComponentPose=@n34, Alpha, OverrideWorldGravity=@n33, ExternalForce) -> (Pose) {
    bClampLinearTranslationLimitToRefPose=true
    bOverrideWorldGravity=true
    overridePhysicsAsset=&F_MED_SuperNovaTaro_Long_FaceAcc_FR_Physics
    simulationSpace=BaseBoneSpace
    alphaBoolBlend {blendOption=Linear}
    baseBoneRef {boneName=spine_05}
    componentAppliedLinearAccClamp=<3500 3500 3500>
    componentLinearAccScale=<0.5 0.5 0.5>
    componentLinearVelScale=<0.5 0.5 0.5>
    overrideWorldGravity {z=-980}
   }
   n36 = RigidBody!(
    ComponentPose=@n35
    bAlphaBoolEnabled:bool=Get backpack() -> (backpack:bool)
    OverrideWorldGravity="0,0,-980"
    ExternalForce
    ComponentLinearAccScale:Vector=Get ACC()
    ComponentLinearVelScale:Vector=Get VEL()
   ) -> (Pose) {
    alphaInputType=Bool
    bClampLinearTranslationLimitToRefPose=true
    bOverrideWorldGravity=true
    overridePhysicsAsset=&F_MED_SuperNovaTaro_Long_FaceAcc_Physics
    simulationSpace=BaseBoneSpace
    alphaBoolBlend {blendOption=Linear}
    baseBoneRef {boneName=spine_05}
    componentAppliedLinearAccClamp=<3500 3500 3500>
    componentLinearAccScale=<0.5 0.5 0.5>
    componentLinearVelScale=<0.5 0.5 0.5>
    overrideWorldGravity {z=-980}
   }
   n37 = "Constraint - dyn_hair_long_bk_l_03 (P)"(ComponentPose=@n36, Alpha=0.5, ConstraintWeights_0=1) {
    alphaBoolBlend {blendOption=Linear}
    boneToModify {boneName=dyn_hair_long_bk_l_03}
    constraintSetup [{targetBone {boneName=dyn_hair_long_bk_03}}]
    constraintWeights [1]
   }
   n38 = "Constraint - dyn_hair_long_bk_l_04 (P)"(ComponentPose=@n37, Alpha=0.5, ConstraintWeights_0=1) {
    alphaBoolBlend {blendOption=Linear}
    boneToModify {boneName=dyn_hair_long_bk_l_04}
    constraintSetup [{targetBone {boneName=dyn_hair_long_bk_04}}]
    constraintWeights [1]
   }
   n39 = "Constraint - dyn_hair_long_bk_lb_04 (P)"(ComponentPose=@n38, Alpha=0.3, ConstraintWeights_0=1) {
    alphaBoolBlend {blendOption=Linear}
    boneToModify {boneName=dyn_hair_long_bk_lb_04}
    constraintSetup [{targetBone {boneName=dyn_hair_long_bk_l_03}}]
    constraintWeights [1]
   }
   n40 = "Constraint - dyn_hair_long_bk_l_05 (P)"(ComponentPose=@n39, Alpha=0.6, ConstraintWeights_0=1) {
    alphaBoolBlend {blendOption=Linear}
    boneToModify {boneName=dyn_hair_long_bk_l_05}
    constraintSetup [{targetBone {boneName=dyn_hair_long_bk_05}}]
    constraintWeights [1]
   }
   n41 = "Constraint - dyn_hair_long_bk_lb_05 (P)"(ComponentPose=@n40, Alpha=0.3, ConstraintWeights_0=1) {
    alphaBoolBlend {blendOption=Linear}
    boneToModify {boneName=dyn_hair_long_bk_lb_05}
    constraintSetup [{targetBone {boneName=dyn_hair_long_bk_l_04}}]
    constraintWeights [1]
   }
   n42 = "Constraint - dyn_hair_long_bk_r_03 (P)"(ComponentPose=@n41, Alpha=0.3, ConstraintWeights_0=1) {
    alphaBoolBlend {blendOption=Linear}
    boneToModify {boneName=dyn_hair_long_bk_r_03}
    constraintSetup [{targetBone {boneName=dyn_hair_long_bk_03}}]
    constraintWeights [1]
   }
   n43 = "Constraint - dyn_hair_long_bk_r_04 (P)"(ComponentPose=@n42, Alpha=0.4, ConstraintWeights_0=1) {
    alphaBoolBlend {blendOption=Linear}
    boneToModify {boneName=dyn_hair_long_bk_r_04}
    constraintSetup [{targetBone {boneName=dyn_hair_long_bk_04}}]
    constraintWeights [1]
   }
   n44 = "Constraint - dyn_hair_long_bk_rb_04 (P)"(ComponentPose=@n43, Alpha=0.3, ConstraintWeights_0=1) {
    alphaBoolBlend {blendOption=Linear}
    boneToModify {boneName=dyn_hair_long_bk_rb_04}
    constraintSetup [{targetBone {boneName=dyn_hair_long_bk_r_04}}]
    constraintWeights [1]
   }
   n45 = "Constraint - dyn_hair_long_bk_r_05 (P)"(ComponentPose=@n44, Alpha=0.5, ConstraintWeights_0=1) {
    alphaBoolBlend {blendOption=Linear}
    boneToModify {boneName=dyn_hair_long_bk_r_05}
    constraintSetup [{targetBone {boneName=dyn_hair_long_bk_05}}]
    constraintWeights [1]
   }
   n46 = "Constraint - dyn_hair_long_bk_rb_05 (P)"(ComponentPose=@n45, Alpha=0.3, ConstraintWeights_0=1) {
    alphaBoolBlend {blendOption=Linear}
    boneToModify {boneName=dyn_hair_long_bk_rb_05}
    constraintSetup [{targetBone {boneName=dyn_hair_long_bk_r_04}}]
    constraintWeights [1]
   }
   n47 = Control Rig!(
    Source=Blend Poses by bool(
     BlendPose_0=Component To Local(ComponentPose=@n18)
     BlendPose_1=Component To Local(ComponentPose=@n46)
     BlendTime_0=0.2
     BlendTime_1=0.3
     bActiveValue=Get bIsCrouching()
    ) {blendType=Linear}
    Alpha
    Crouching:bool=Get bIsCrouching()
    Moving:double=Get Moving() -> (Moving:double)
   ) -> (Pose) {
    alphaBoolBlend {blendOption=Linear}
    controlRigAssetReference {blueprintRigClass=&/F_MED_SuperNovaTaro_Long_FaceAcc_CtrlRig.F_MED_SuperNovaTaro_Long_FaceAcc_CtrlRig_C}
   }
   Output Pose(Result=@n47) {layerGroup=None}
  ]}
 EventGraph {nodes [
   Event BlueprintUpdateAnimation()
   > n1 = Sequence()
   Set Moving(execute=@n1.then_1, Moving:double=SelectFloat(A=1, B=0, bPickA=Get bIsMoving2D())) -> (Output_Get:double)
   > Set Skydive(Skydive:double=SelectFloat(A=1, B=0, bPickA=Get bIsMoving2D())) -> (Output_Get:double)
   n2 = Gravity Override State Function DEPRECATED(
    execute=@n1.then_0
    CharacterPartAnimInstance=Self-Reference()
    GameTimeInSeconds=GetGameTimeInSeconds()
    GravityOverrideParams=Get Gravity Override Params()
   )
   > Set GravityOverride(GravityOverride=@n2)
   > n3 = Branch(Condition=CheckForBackbling())
   > Set ACC(ACC:Vector="0, 0, 0") -> (Output_Get:Vector)
   > Set VEL(VEL:Vector="0, 0, 0") -> (Output_Get:Vector)
   > Set backpack(backpack:bool=true) -> (Output_Get:bool)
   Set ACC(execute=@n3.else, ACC="0.5,0.5,0.5")
   > Set VEL(VEL="0.5,0.5,0.5")
   > Set backpack(backpack=false)
   Event BlueprintInitializeAnimation()
   > Set Skel_Mesh(Skel_Mesh=GetOwnerSkeletalMeshComponent())
  ]}
]
variables [
 ACC {type=Vector}
 backpack {type=boolean}
 Gravity Override Params {type=GravityOverrideParamsStruct}
 GravityOverride {type=Vector}
 Moving {type=number}
 Skel_Mesh {type=SkeletalMeshComponent}
 Skydive {type=number}
 VEL {type=Vector}
]
